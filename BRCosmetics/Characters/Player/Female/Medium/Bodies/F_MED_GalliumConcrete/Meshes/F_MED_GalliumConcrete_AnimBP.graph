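# Post-process style AnimBP for the F_MED_GalliumConcrete female cosmetic body, adding secondary motion and deformation fixes to the player's pose.
# The copied parent pose feeds a Control Rig driven by weapon, movement and emote flags set in the EventGraph.
# AnimDynamics on wristbands, zipper, sleeves and shirt follows, then additive bone offsets.
# Outside the front end, it also feeds material WPO cloth simulation from foot and calf bones, raising wind while skydive-floating or parachuting.
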AnimBlueprint /BRCosmetics/Characters/Player/Female/Medium/Bodies/F_MED_GalliumConcrete/Meshes/F_MED_GalliumConcrete_AnimBP
target_skeleton=&/F_MED_GalliumConcrete_Skeleton
graphs [
 AnimGraph {nodes [
   n1 = "Transform (Modify) Bone - Bone: dyn_Sleeve_under_l"(
    ComponentPose=Local To Component(LocalPose=Use cached pose 'CRCache_zipline'())
    Alpha=1
    Translation="0,0,0"
    Rotation="0,0,0"
    Scale="1,1,1"
   ) {rotationMode=BMM_Additive, translationMode=BMM_Additive, alphaBoolBlend {blendOption=Linear}, boneToModify {boneName=dyn_Sleeve_under_l}}
   n2 = "Transform (Modify) Bone - Bone: dyn_Sleeve_top_l"(ComponentPose=@n1, Alpha=1, Translation="0,0,0", Rotation="0,0,0", Scale="1,1,1") {rotationMode=BMM_Additive, translationMode=BMM_Additive, alphaBoolBlend {blendOption=Linear}, boneToModify {boneName=dyn_Sleeve_top_l}}
   n3 = "Transform (Modify) Bone - Bone: dyn_Sleeve_back_l"(ComponentPose=@n2, Alpha=1, Translation="0,0,0", Rotation="0,0,0", Scale="1,1,1") {rotationMode=BMM_Additive, translationMode=BMM_Additive, alphaBoolBlend {blendOption=Linear}, boneToModify {boneName=dyn_Sleeve_back_l}}
   n4 = "Transform (Modify) Bone - Bone: dyn_Sleeve_front_l"(ComponentPose=@n3, Alpha=1, Translation="0,0,0", Rotation="0,0,0", Scale="1,1,1") {rotationMode=BMM_Additive, translationMode=BMM_Additive, alphaBoolBlend {blendOption=Linear}, boneToModify {boneName=dyn_Sleeve_front_l}}
   n5 = "Transform (Modify) Bone - Bone: dyn_Sleeve_under_r"(ComponentPose=@n4, Alpha=1, Translation="0,0,0", Rotation="0,0,0", Scale="1,1,1") {rotationMode=BMM_Additive, translationMode=BMM_Additive, alphaBoolBlend {blendOption=Linear}, boneToModify {boneName=dyn_Sleeve_under_r}}
   n6 = "Transform (Modify) Bone - Bone: dyn_Sleeve_top_r"(ComponentPose=@n5, Alpha=1, Translation="0,0,0", Rotation="0,0,0", Scale="1,1,1") {rotationMode=BMM_Additive, translationMode=BMM_Additive, alphaBoolBlend {blendOption=Linear}, boneToModify {boneName=dyn_Sleeve_top_r}}
   n7 = "Transform (Modify) Bone - Bone: dyn_Sleeve_back_r"(ComponentPose=@n6, Alpha=1, Translation="0,0,0", Rotation="0,0,0", Scale="1,1,1") {rotationMode=BMM_Additive, translationMode=BMM_Additive, alphaBoolBlend {blendOption=Linear}, boneToModify {boneName=dyn_Sleeve_back_r}}
   n8 = "Transform (Modify) Bone - Bone: dyn_Sleeve_front_r"(ComponentPose=@n7, Alpha=1, Translation="0,0,0", Rotation="0,0,0", Scale="1,1,1") {rotationMode=BMM_Additive, translationMode=BMM_Additive, alphaBoolBlend {blendOption=Linear}, boneToModify {boneName=dyn_Sleeve_front_r}}
   n9 = "Transform (Modify) Bone - Bone: base_pack_01_r"(ComponentPose=@n8, Alpha=1, Translation="0,0,1", Rotation="0,30,5", Scale="1,1,1") {
    rotationMode=BMM_Additive
    rotationSpace=BCS_BoneSpace
    translationMode=BMM_Additive
    translationSpace=BCS_BoneSpace
    alphaBoolBlend {blendOption=Linear}
    alphaScaleBiasClamp {bInterpResult=true, interpSpeedDecreasing=3, interpSpeedIncreasing=3}
    boneToModify {boneName=base_pack_01_r}
    rotation {roll=5, yaw=30}
    translation {z=1}
   }
   n10 = "Transform (Modify) Bone - Bone: base_wristband_01_r"(ComponentPose=@n9, Alpha=1, Translation="0,0,0", Rotation="20,-5,0", Scale="1,1,1") {
    rotationMode=BMM_Additive
    rotationSpace=BCS_BoneSpace
    scaleMode=BMM_Additive
    translationMode=BMM_Additive
    translationSpace=BCS_BoneSpace
    alphaBoolBlend {blendOption=Linear}
    alphaScaleBiasClamp {bInterpResult=true, interpSpeedDecreasing=3, interpSpeedIncreasing=3}
    boneToModify {boneName=base_wristband_01_r}
    rotation {pitch=20, yaw=-5}
   }
   n11 = "Transform (Modify) Bone - Bone: dyn_wristband_02_r"(ComponentPose=@n10, Alpha=1, Translation="0,0,0.3", Rotation="20,0,0", Scale="1.2,1.2,1.2") {
    rotationMode=BMM_Additive
    rotationSpace=BCS_BoneSpace
    scaleMode=BMM_Additive
    translationMode=BMM_Additive
    translationSpace=BCS_BoneSpace
    alphaBoolBlend {blendOption=Linear}
    alphaScaleBiasClamp {bInterpResult=true, interpSpeedDecreasing=3, interpSpeedIncreasing=3}
    boneToModify {boneName=dyn_wristband_02_r}
    rotation {pitch=20}
    scale=<1.2 1.2 1.2>
    translation {z=0.3}
   }
   Output Pose(Result=Component To Local(ComponentPose=@n11)) {layerGroup=None}
   n12 = Control Rig!(
    Source=Copy Pose From Mesh() {bUseAttachedParent=true}
    Alpha
    OpenedParachute:double=Get ParachuteOpened() -> (ParachuteOpened:double)
    OnBaseballSliding:double=Get BaseballSliding() -> (BaseballSliding:double)
    IsCelebration:double=Get IsCelebrationEmote() -> (IsCelebrationEmote:double)
    IsSliding:double=Get IsSlidingOn() -> (IsSlidingOn:double)
    Crouching:double=Get IsCrouchingOn() -> (IsCrouchingOn:double)
    Rifle01_out_deform:double=Get Rifle01_out() -> (Rifle01_out:double)
    Shotgun01_out_deform:double=Get Shotgun01_out() -> (Shotgun01_out:double)
    Pistol01_out_deform:double=Get Pistol01_out() -> (Pistol01_out:double)
    Pistol02_out_deform:double=Get Pistol02_out() -> (Pistol02_out:double)
    Rocket_out_deform:double=Get Rocket_out() -> (Rocket_out:double)
    GoinCommando_deform:double=Get GoinCommando_out() -> (GoinCommando_out:double)
    Floating_out_deform:double=Get FloatingMode_Out() -> (FloatingMode_Out:double)
    MeleeAttack_out_deform:double=Get MeleeAttack_out() -> (MeleeAttack_out:double)
    Ballet_out_deform:double=Get IsBalletEmote() -> (IsBalletEmote:double)
   ) -> (Pose) {
    alphaBoolBlend {blendOption=Linear}
    controlRigAssetReference {blueprintRigClass=&/F_MED_GalliumConcrete_CtrlRig.F_MED_GalliumConcrete_CtrlRig_C}
   }
   Save cached pose 'CrCache_input'(Pose=@n12) {cacheName=CrCache_input}
   n13 = Anim Dynamics - Bone: dyn_wristband_01_r!(
    SimulationSpace:AnimPhysSimSpaceType Enum=Component
    ComponentPose=Local To Component(LocalPose=Use cached pose 'CrCache_input'())
    Alpha=1
    ExternalForce="0,0,0"
   ) -> (Pose) {
    angularDampingOverride=1
    angularSpringConstant=20000
    bAngularSpring=true
    bOverrideAngularDamping=true
    bUseGravityOverride=true
    gravityScale=0.35
    alphaBoolBlend {blendOption=Linear}
    boundBone {boneName=dyn_wristband_01_r}
    componentLinearAccScale=<1 1 1>
    componentLinearVelScale=<1 1 1>
    gravityOverride {z=-980}
    physicsBodyDefinitions [{
      boundBone {boneName=dyn_wristband_01_r}
      boxExtents=<20 30 30>
      constraintSetup {angularLimitsMax=<15 15 15>, angularLimitsMin=<-15 -15 -15>, angularTarget=<1 0 0>}
      localJointOffset=<4 0 0>
     }]
   }
   n14 = Anim Dynamics - Bone: dyn_wristband_02_r!(SimulationSpace:AnimPhysSimSpaceType Enum=Component, ComponentPose=@n13, Alpha=1, ExternalForce="0,0,0") -> (Pose) {
    angularDampingOverride=1
    angularSpringConstant=20000
    bAngularSpring=true
    bOverrideAngularDamping=true
    bUseGravityOverride=true
    gravityScale=0.35
    alphaBoolBlend {blendOption=Linear}
    boundBone {boneName=dyn_wristband_02_r}
    componentLinearAccScale=<1 1 1>
    componentLinearVelScale=<1 1 1>
    gravityOverride {z=-980}
    physicsBodyDefinitions [{
      boundBone {boneName=dyn_wristband_02_r}
      boxExtents=<20 30 30>
      constraintSetup {angularLimitsMax=<15 15 15>, angularLimitsMin=<-15 -15 -15>, angularTarget=<1 0 0>}
      localJointOffset=<4 0 0>
     }]
   }
   n15 = Anim Dynamics - Bone: dyn_zipper_01_r!(SimulationSpace:AnimPhysSimSpaceType Enum=Component, ComponentPose=@n14, Alpha=0.7, ExternalForce="0,0,0") -> (Pose) {
    angularDampingOverride=1
    angularSpringConstant=20000
    bAngularSpring=true
    bOverrideAngularDamping=true
    bUseGravityOverride=true
    gravityScale=0.35
    alphaBoolBlend {blendOption=Linear}
    boundBone {boneName=dyn_zipper_01_r}
    componentLinearAccScale=<1 1 1>
    componentLinearVelScale=<1 1 1>
    gravityOverride {z=-980}
    physicsBodyDefinitions [{
      boundBone {boneName=dyn_zipper_01_r}
      boxExtents=<15 15 15>
      constraintSetup {angularLimitsMax=<20 10 40>, angularLimitsMin=<-20 -20 -40>, angularTarget=<1 0 0>}
      localJointOffset=<4 0 0>
     }]
   }
   n16 = Anim Dynamics - Bone: dyn_Sleeve_under_l!(SimulationSpace:AnimPhysSimSpaceType Enum=Component, ComponentPose=@n15, Alpha=1, ExternalForce="0,0,0") -> (Pose) {
    angularDampingOverride=1
    angularSpringConstant=20000
    bAngularSpring=true
    bOverrideAngularDamping=true
    bUseGravityOverride=true
    gravityScale=0.35
    alphaBoolBlend {blendOption=Linear}
    boundBone {boneName=dyn_Sleeve_under_l}
    componentLinearAccScale=<0.4 0.35 0.35>
    componentLinearVelScale=<0.35 0.35 0.35>
    gravityOverride {z=-980}
    physicsBodyDefinitions [{
      boundBone {boneName=dyn_Sleeve_under_l}
      boxExtents=<20 30 30>
      constraintSetup {angularLimitsMax=<7 7 7>, angularLimitsMin=<-7 -7 -7>, angularTarget=<1 0 0>}
      localJointOffset=<4 0 0>
     }]
   }
   n17 = Anim Dynamics - Bone: dyn_Sleeve_back_l!(SimulationSpace:AnimPhysSimSpaceType Enum=Component, ComponentPose=@n16, Alpha=1, ExternalForce="0,0,0") -> (Pose) {
    angularDampingOverride=1
    angularSpringConstant=20000
    bAngularSpring=true
    bOverrideAngularDamping=true
    bUseGravityOverride=true
    gravityScale=0.35
    alphaBoolBlend {blendOption=Linear}
    boundBone {boneName=dyn_Sleeve_back_l}
    componentLinearAccScale=<0.4 0.35 0.35>
    componentLinearVelScale=<0.35 0.35 0.35>
    gravityOverride {z=-980}
    physicsBodyDefinitions [{
      boundBone {boneName=dyn_Sleeve_back_l}
      boxExtents=<20 30 30>
      constraintSetup {angularLimitsMax=<0 0 0>, angularLimitsMin=<-20 -20 -20>, angularTarget=<1 0 0>}
      localJointOffset=<4 0 0>
     }]
   }
   n18 = Anim Dynamics - Bone: dyn_Sleeve_top_l!(SimulationSpace:AnimPhysSimSpaceType Enum=Component, ComponentPose=@n17, Alpha=1, ExternalForce="0,0,0") -> (Pose) {
    angularDampingOverride=1
    angularSpringConstant=20000
    bAngularSpring=true
    bOverrideAngularDamping=true
    bUseGravityOverride=true
    gravityScale=0.35
    alphaBoolBlend {blendOption=Linear}
    boundBone {boneName=dyn_Sleeve_top_l}
    componentLinearAccScale=<0.4 0.35 0.35>
    componentLinearVelScale=<0.35 0.35 0.35>
    gravityOverride {z=-980}
    physicsBodyDefinitions [{
      boundBone {boneName=dyn_Sleeve_top_l}
      boxExtents=<20 30 30>
      constraintSetup {angularLimitsMax=<0 0 0>, angularLimitsMin=<-8 -8 -1>, angularTarget=<1 0 0>}
      localJointOffset=<-4 0 0>
     }]
   }
   n19 = Anim Dynamics - Bone: dyn_Sleeve_under_r!(SimulationSpace:AnimPhysSimSpaceType Enum=Component, ComponentPose=@n18, Alpha=0, ExternalForce="0,0,0") -> (Pose) {
    angularDampingOverride=1
    angularSpringConstant=20000
    bAngularSpring=true
    bOverrideAngularDamping=true
    bUseGravityOverride=true
    gravityScale=0.35
    alphaBoolBlend {blendOption=Linear}
    boundBone {boneName=dyn_Sleeve_under_r}
    componentLinearAccScale=<0.4 0.35 0.35>
    componentLinearVelScale=<0.35 0.35 0.35>
    gravityOverride {z=-980}
    physicsBodyDefinitions [{
      boundBone {boneName=dyn_Sleeve_under_r}
      boxExtents=<20 30 30>
      constraintSetup {angularLimitsMax=<7 7 7>, angularLimitsMin=<-7 -7 -7>, angularTarget=<1 0 0>}
      localJointOffset=<4 0 0>
     }]
   }
   n20 = Anim Dynamics - Bone: dyn_Sleeve_back_r!(SimulationSpace:AnimPhysSimSpaceType Enum=Component, ComponentPose=@n19, Alpha=0, ExternalForce="0,0,0") -> (Pose) {
    angularDampingOverride=1
    angularSpringConstant=20000
    bAngularSpring=true
    bOverrideAngularDamping=true
    bUseGravityOverride=true
    gravityScale=0.35
    alphaBoolBlend {blendOption=Linear}
    boundBone {boneName=dyn_Sleeve_back_r}
    componentLinearAccScale=<0.4 0.35 0.35>
    componentLinearVelScale=<0.35 0.35 0.35>
    gravityOverride {z=-980}
    physicsBodyDefinitions [{
      boundBone {boneName=dyn_Sleeve_back_r}
      boxExtents=<20 30 30>
      constraintSetup {angularLimitsMax=<0 0 0>, angularLimitsMin=<-20 -20 -20>, angularTarget=<1 0 0>}
      localJointOffset=<4 0 0>
     }]
   }
   n21 = Anim Dynamics - Bone: dyn_Sleeve_top_r!(SimulationSpace:AnimPhysSimSpaceType Enum=Component, ComponentPose=@n20, Alpha=0, ExternalForce="0,0,0") -> (Pose) {
    angularDampingOverride=1
    angularSpringConstant=20000
    bAngularSpring=true
    bOverrideAngularDamping=true
    bUseGravityOverride=true
    gravityScale=0.35
    alphaBoolBlend {blendOption=Linear}
    boundBone {boneName=dyn_Sleeve_top_r}
    componentLinearAccScale=<0.4 0.35 0.35>
    componentLinearVelScale=<0.35 0.35 0.35>
    gravityOverride {z=-980}
    physicsBodyDefinitions [{
      boundBone {boneName=dyn_Sleeve_top_r}
      boxExtents=<20 30 30>
      constraintSetup {angularLimitsMax=<0 0 0>, angularLimitsMin=<-20 -20 -10>, angularTarget=<1 0 0>}
      localJointOffset=<-4 0 0>
     }]
   }
   n22 = Anim Dynamics - Bone: dyn_Shirt_back_01_l!(SimulationSpace:AnimPhysSimSpaceType Enum=Component, ComponentPose=@n21, Alpha=0.3, ExternalForce="0,0,0") -> (Pose) {
    angularDampingOverride=1
    angularSpringConstant=20000
    bAngularSpring=true
    bOverrideAngularDamping=true
    bUseGravityOverride=true
    gravityScale=0.35
    alphaBoolBlend {blendOption=Linear}
    boundBone {boneName=dyn_Shirt_back_01_l}
    componentLinearAccScale=<0.4 0.35 0.35>
    componentLinearVelScale=<0.35 0.35 0.35>
    gravityOverride {z=-980}
    physicsBodyDefinitions [{
      boundBone {boneName=dyn_Shirt_back_01_l}
      boxExtents=<15 25 25>
      constraintSetup {angularLimitsMax=<30 0 30>, angularLimitsMin=<-30 0 -30>, angularTarget=<1 0 0>}
      localJointOffset=<5 0 0>
     }]
   }
   n23 = Anim Dynamics - Bone: dyn_Shirt_back_02_l!(SimulationSpace:AnimPhysSimSpaceType Enum=Component, ComponentPose=@n22, Alpha=0.3, ExternalForce="0,0,0") -> (Pose) {
    angularDampingOverride=1
    angularSpringConstant=20000
    bAngularSpring=true
    bOverrideAngularDamping=true
    bUseGravityOverride=true
    gravityScale=0.35
    alphaBoolBlend {blendOption=Linear}
    boundBone {boneName=dyn_Shirt_back_02_l}
    componentLinearAccScale=<0.4 0.35 0.35>
    componentLinearVelScale=<0.35 0.35 0.35>
    gravityOverride {z=-980}
    physicsBodyDefinitions [{
      boundBone {boneName=dyn_Shirt_back_02_l}
      boxExtents=<15 25 25>
      constraintSetup {angularLimitsMax=<30 0 30>, angularLimitsMin=<-30 0 -30>, angularTarget=<1 0 0>}
      localJointOffset=<5 0 0>
     }]
   }
   n24 = Anim Dynamics - Bone: dyn_Shirt_back_01_r!(SimulationSpace:AnimPhysSimSpaceType Enum=Component, ComponentPose=@n23, Alpha=0.3, ExternalForce="0,0,0") -> (Pose) {
    angularDampingOverride=1
    angularSpringConstant=20000
    bAngularSpring=true
    bOverrideAngularDamping=true
    bUseGravityOverride=true
    gravityScale=0.35
    alphaBoolBlend {blendOption=Linear}
    boundBone {boneName=dyn_Shirt_back_01_r}
    componentLinearAccScale=<0.4 0.35 0.35>
    componentLinearVelScale=<0.35 0.35 0.35>
    gravityOverride {z=-980}
    physicsBodyDefinitions [{
      boundBone {boneName=dyn_Shirt_back_01_r}
      boxExtents=<15 25 25>
      constraintSetup {angularLimitsMax=<30 0 30>, angularLimitsMin=<-30 0 -30>, angularTarget=<1 0 0>}
      localJointOffset=<5 0 0>
     }]
   }
   n25 = Anim Dynamics - Bone: dyn_Shirt_back_02_r!(SimulationSpace:AnimPhysSimSpaceType Enum=Component, ComponentPose=@n24, Alpha=0.3, ExternalForce="0,0,0") -> (Pose) {
    angularDampingOverride=1
    angularSpringConstant=20000
    bAngularSpring=true
    bOverrideAngularDamping=true
    bUseGravityOverride=true
    gravityScale=0.35
    alphaBoolBlend {blendOption=Linear}
    boundBone {boneName=dyn_Shirt_back_02_r}
    componentLinearAccScale=<0.4 0.35 0.35>
    componentLinearVelScale=<0.35 0.35 0.35>
    gravityOverride {z=-980}
    physicsBodyDefinitions [{
      boundBone {boneName=dyn_Shirt_back_02_r}
      boxExtents=<15 25 25>
      constraintSetup {angularLimitsMax=<30 0 30>, angularLimitsMin=<-30 0 -30>, angularTarget=<1 0 0>}
      localJointOffset=<5 0 0>
     }]
   }
   Save cached pose 'CRCache_zipline'(Pose=Component To Local(ComponentPose=@n25)) {cacheName=CRCache_zipline}
  ]}
 EventGraph {nodes [
   Event BlueprintInitializeAnimation()
   > Set SkelMesh(SkelMesh=GetOwnerSkeletalMeshComponent())
   n1 = TryGetPawnOwner()
   > Is Valid(InputObject=@n1)
   > n2 = Cast To FortPlayerPawnAthena(Object=@n1)
   > n3 = Set PlayerPawn(PlayerPawn:Fort Player Pawn Athena Object Reference=@n2) -> (Output_Get:Fort Player Pawn Athena Object Reference)
   > Set Body_Mesh(Body_Mesh=GetSkeletalMeshForPartType(self=@n3))
   n4 = Event BlueprintUpdateAnimation()
   > n5 = Sequence()
   > Is Valid(InputObject=Get PlayerPawn() -> (PlayerPawn:Fort Player Pawn Athena Object Reference))
   > n6 = Branch(Condition=Get bIsInFrontEnd())
   Set WPO Globals Function(
    execute=@n6.else
    Material=Get Mat Instance Dynamic()
    Overlay Material=Get Overlay Mat Instance Dynamic()
    Skeletal Mesh=Get SkelMesh()
    "Local Noise Transform  Location"=Get Local Noise Transform Location() -> (Local Noise Transform Location:Vector)
    Local Noise Offset=Get Local Noise Offset()
   )
   > Set WPO Regional Function!(
    execute
    Material:Material Instance Dynamic Object Reference=Get Mat Instance Dynamic()
    Overlay Material:Material Instance Dynamic Object Reference=Get Overlay Mat Instance Dynamic()
    Skeletal Mesh:Skeletal Mesh Component Object Reference=Get SkelMesh()
    Velocity Transform:Name=foot_r
    Velocity Interp Speed:double=Get Velocity Interp Speed()
    Collision Transform:Name=calf_r
    Collision Transform Axis:EAxis Enum
    Invert Collision Transform Axis:bool=false
    Delta Time X:double=@n4.DeltaTimeX
    Previous Velocity Transform Location:"Vector (by ref)"=Get R1 Velocity Transform Location()
    Previous Velocity Transform Velocity:"Vector (by ref)"=Get R1 Velocity Transform Velocity()
    Region:int=1
   ) -> (then)
   > Set WPO Regional Function(
    Material=Get Mat Instance Dynamic()
    Overlay Material=Get Overlay Mat Instance Dynamic()
    Skeletal Mesh=Get SkelMesh()
    Velocity Transform=foot_l
    Velocity Interp Speed=Get Velocity Interp Speed()
    Collision Transform=calf_l
    Invert Collision Transform Axis=true
    Delta Time X=@n4.DeltaTimeX
    Previous Velocity Transform Location=Get R2 Velocity Transform Location()
    Previous Velocity Transform Velocity=Get R2 Velocity Transform Velocity()
    Region=2
   )
   > n7 = Branch(Condition=OR Boolean(A=Get bIsSkydiveFloating(), B=Get bIsParachuteOpened()))
   > Set WPO Pho Wind Intensity(WPO Pho Wind Intensity=1.5)
   > n8 = Set WPO Pho Turbulence(WPO Pho Turbulence:double=1.7) -> (Output_Get:double)
   SetScalarParameterValue(
    execute=[@n10.then, @n8.then]
    self=[Get Mat Instance Dynamic(), Get Overlay Mat Instance Dynamic()]
    ParameterName=ClothSim_WindStrength
    Value=Get WPO Pho Wind Intensity()
   )
   n9 = Get WPO Pho Turbulence() -> (WPO Pho Turbulence:double)
   > SetScalarParameterValue(self=[Get Mat Instance Dynamic(), Get Overlay Mat Instance Dynamic()], ParameterName=ClothSim_R1_TurbulenceStrength, Value=@n9)
   > SetScalarParameterValue(self=[Get Overlay Mat Instance Dynamic(), Get Mat Instance Dynamic()], ParameterName=ClothSim_R2_TurbulenceStrength, Value=@n9)
   Set WPO Pho Wind Intensity(execute=@n7.else, WPO Pho Wind Intensity=0)
   > n10 = Set WPO Pho Turbulence(WPO Pho Turbulence=0.1)
   n11 = Sequence!(execute=@n5.then_1) -> (then_0, then_1, then_2:Exec, then_3:Exec, then_4:Exec)
   > n12 = Cast To FortPlayerAnimInstance(Object=GetAnimInstance(self=GetOwnerSkeletalMeshComponent()))
   > Set Rifle01(Rifle01:bool=false) -> (Output_Get:bool)
   > Set Shotgun01(Shotgun01:bool=false) -> (Output_Get:bool)
   > Set Pistol01(Pistol01:bool=false) -> (Output_Get:bool)
   > Set Rocket(Rocket:bool=false) -> (Output_Get:bool)
   > Set GoinCommando(GoinCommando:bool=false) -> (Output_Get:bool)
   > n13 = Switch on EFortWeaponCoreAnimation!(execute, Selection:EFortWeaponCoreAnimation Enum=Get WeaponCoreAnim(self=@n12)) -> (
    Melee:Exec
    Pistol:Exec
    Shotgun:Exec
    PaperBlueprint:Exec
    Rifle:Exec
    MeleeOneHand:Exec
    MachinePistol:Exec
    RocketLauncher:Exec
    GrenadeLauncher:Exec
    GoingCommando:Exec
    AssaultRifle:Exec
    TacticalShotgun:Exec
    SniperRifle:Exec
    TrapPlacement:Exec
    ShoulderLauncher:Exec
    AbilityDecoTool:Exec
    Crossbow:Exec
    C4:Exec
    RemoteControl:Exec
    DualWield:Exec
    AR_BullPup:Exec
    AR_ForwardGrip:Exec
    MedPackPaddles:Exec
    SMG_P90:Exec
    AR_DrumGun:Exec
    Consumable_Small:Exec
    Consumable_Large:Exec
    Balloon:Exec
    MountedTurret:Exec
    CreativeTool:Exec
    ExplosiveBow:Exec
    AshtonIndigo:Exec
    AshtonChicago:Exec
    MeleeDualWield:Exec
    Unarmed:Exec
   )
   Branch(execute=@n13.AssaultRifle)
   > Set Rifle01(Rifle01=true)
   Branch(execute=@n13.GoingCommando)
   > Set GoinCommando(GoinCommando=true)
   Branch(execute=@n13.Pistol)
   > Set Pistol01(Pistol01=true)
   Branch(execute=@n13.RocketLauncher)
   > Set Rocket(Rocket=true)
   Branch(execute=@n13.Shotgun)
   > Set Shotgun01(Shotgun01=true)
   n14 = Event BlueprintProcessAnimInputEvent()
   n15 = Branch(execute=@n11.then_1, Condition=IsNamed(self=@n14.AnimInputEvent, InEventName=aie.Player.Zipline.Entered))
   > Set OnZipline(OnZipline:bool=1) -> (Output_Get:bool)
   > Set ZiplineAlpha(ZiplineAlpha:double=0) -> (Output_Get:double)
   Branch(
    execute=@n15.else
    Condition=OR Boolean!(
     A=IsNamed(self=@n14.AnimInputEvent, InEventName=aie.Player.Zipline.Exited)
     B=Get bIsMoving2D()
     C:bool=Get bIsJumping()
     D:bool=Get bIsInFrontEnd()
    ) -> (ReturnValue)
   )
   > Set OnZipline(OnZipline=false)
   > Set ZiplineAlpha(ZiplineAlpha=1)
   n16 = GetRealTimeSeconds()
   n17 = "Sin (Radians)"(A=float * float(A=float * float(A=1.1, B=1.1), B=@n16))
   n18 = "Sin (Radians)"(A=float * float(A=float * float(A=0.9, B=1.1), B=@n16))
   n19 = "Sin (Radians)"(A=float * float(A=@n16, B=float * float(A=1.2, B=5)))
   Set SleevesRotation(
    execute=@n11.then_2
    SleevesRotation:Rotator=MakeRotator(
     Roll=float * float(A=@n17, B=0.15)
     Pitch=float * float(A=@n18, B=0.3)
     Yaw=Remap Float(Input=@n19, OldRange=(X=-1,Y=1), NewRange=(X=-7,Y=7)) -> (Output=~)
    )
   ) -> (Output_Get:Rotator)
   n20 = Get bIsOnGrindRail()
   n21 = Get OnZipline() -> (OnZipline:bool)
   > Set Rotation_Alpha(Rotation_Alpha:double="To Float (Boolean)"(InBool=OR Boolean!(A=@n20, B=@n21) -> (ReturnValue))) -> (Output_Get:double)
   > Set BaseballSliding(BaseballSliding:double="To Float (Boolean)"(InBool=Get bIsBaseballSliding())) -> (Output_Get:double)
   > Set IsSlidingOn(IsSlidingOn:double="To Float (Boolean)"(InBool=Get bIsSliding())) -> (Output_Get:double)
   n22 = Get bIsCrouching()
   n23 = Get bIsCrouchSprinting()
   > Set IsCrouchingOn(IsCrouchingOn:double="To Float (Boolean)"(InBool=OR Boolean(A=@n22, B=@n23))) -> (Output_Get:double)
   n24 = Get bIsJumping()
   n25 = Get bIsFalling()
   > Set IsJumpOn(IsJumpOn:double="To Float (Boolean)"(InBool=OR Boolean(A=@n24, B=@n25))) -> (Output_Get:double)
   n26 = Get bIsSkydiveDiveMode()
   n27 = Get bIsSkydiveFloating()
   > Set FloatingMode_Out(FloatingMode_Out:double="To Float (Boolean)"(InBool=OR Boolean(A=@n27, B=@n26))) -> (Output_Get:double)
   > Set MeleeAttack_out(MeleeAttack_out:double="To Float (Boolean)"(InBool=Get bIsPlayingMeleeAnim())) -> (Output_Get:double)
   > Set IsDBNO(IsDBNO:double="To Float (Boolean)"(InBool=Get bIsDBNO())) -> (Output_Get:double)
   n28 = Sequence(execute=@n11.then_3)
   Set Rifle01_out(execute=@n28.then_1, Rifle01_out:double="To Float (Boolean)"(InBool=Get Rifle01() -> (Rifle01:bool))) -> (Output_Get:double)
   > Set Shotgun01_out(Shotgun01_out:double="To Float (Boolean)"(InBool=Get Shotgun01() -> (Shotgun01:bool))) -> (Output_Get:double)
   > Set Pistol01_out(Pistol01_out:double="To Float (Boolean)"(InBool=Get Pistol01() -> (Pistol01:bool))) -> (Output_Get:double)
   > Set Pistol02_out(Pistol02_out:double="To Float (Boolean)"(InBool=Get Pistol01())) -> (Output_Get:double)
   > Set Rocket_out(Rocket_out:double="To Float (Boolean)"(InBool=Get Rocket() -> (Rocket:bool))) -> (Output_Get:double)
   > Set GoinCommando_out(GoinCommando_out:double="To Float (Boolean)"(InBool=Get GoinCommando() -> (GoinCommando:bool))) -> (Output_Get:double)
   n29 = Branch(execute=@n28.then_2, Condition=Get bIsParachuteOpened())
   > Set ParachuteOpened(ParachuteOpened:double=1) -> (Output_Get:double)
   Set ParachuteOpened(execute=@n29.else, ParachuteOpened=0)
   n30 = GetAnimInstance(self=GetOwnerSkeletalMeshComponent())
   Is Valid(exec=@n28.then_3, InputObject=@n30)
   > n31 = Set IsBalletSpin(IsBalletSpin:bool=Montage_IsPlaying(self=@n30, Montage=/BRCosmetics/Animation/Game/MainPlayer/Emotes/BalletSpin/Emote_BalletSpin_CMF_M)) -> (Output_Get:bool)
   > n32 = Branch(Condition=@n31)
   > Set IsBalletEmote(IsBalletEmote:double=1) -> (Output_Get:double)
   Set IsBalletEmote(execute=@n32.else, IsBalletEmote=0)
   n33 = Branch(execute=@n28.then_4, Condition=Get bIsInFrontEnd())
   > Set SimlationSpace(SimlationSpace:AnimPhysSimSpaceType Enum=World) -> (Output_Get:AnimPhysSimSpaceType Enum)
   Set SimlationSpace(execute=@n33.else, SimlationSpace=Component)
   Event OnCharacterCustomizationLoaded()
   > n34 = Set SkelMesh(SkelMesh=GetPartSkeletalMeshComponent())
   > n35 = Get Dynamic Material Instance(SkeletalMeshComponent=@n34)
   > Set Mat Instance Dynamic(Mat Instance Dynamic=@n35)
   > n36 = GetOrCreateDynamicOverlayMaterial(MeshComponent=@n34, MaterialName=None)
   > Set Overlay Mat Instance Dynamic(Overlay Mat Instance Dynamic=@n36)
   n37 = Set Iscelebration(Iscelebration:bool=OR Boolean(
     A=Montage_IsPlaying(self=@n30, Montage=/BRCosmetics/Animation/Game/MainPlayer/Emotes/Emote_Celebration_Female_M)
     B=Montage_IsPlaying(self=@n30, Montage=/BRCosmetics/Animation/Game/MainPlayer/Emotes/Mime/Emote_Mime_CMF_M)
    )) -> (Output_Get:bool)
   > n38 = Branch(Condition=@n37)
   > Set IsCelebrationEmote(IsCelebrationEmote:double=1) -> (Output_Get:double)
   Set IsCelebrationEmote(execute=@n38.else, IsCelebrationEmote=0)
   TryGetPawnOwner()
   Reroute Node(InputPin:Wildcard) -> (OutputPin:Wildcard)
   Get bIsParachuteOpened()
  ]}
]
variables [
 BaseballSliding {type=number}
 Body_Mesh {type=SkeletalMeshComponent}
 FloatingMode_Out {type=number}
 GoinCommando {type=boolean}
 GoinCommando_out {type=number}
 IsBalletEmote {type=number}
 IsBalletSpin {type=boolean}
 Iscelebration {type=boolean}
 IsCelebrationEmote {type=number}
 IsCrouchingOn {type=number}
 IsDBNO {type=number}
 IsJumpOn {type=number}
 Ismoving {type=number}
 IsSlidingOn {type=number}
 Local Noise Offset {type=Vector}
 Local Noise Transform Location {type=Vector}
 Mat Instance Dynamic {type=MaterialInstanceDynamic}
 MeleeAttack_out {type=number}
 OnZipline {type=boolean}
 Overlay Mat Instance Dynamic {type=MaterialInstanceDynamic}
 ParachuteOpened {type=number}
 Pistol01 {type=boolean}
 Pistol01_out {type=number}
 Pistol02_out {type=number}
 PlayerPawn {type=FortPlayerPawnAthena}
 R1 Velocity Transform Location {type=Vector}
 R1 Velocity Transform Velocity {type=Vector}
 R2 Velocity Transform Location {type=Vector}
 R2 Velocity Transform Velocity {type=Vector}
 Rifle01 {type=boolean}
 Rifle01_out {type=number}
 Rocket {type=boolean}
 Rocket_out {type=number}
 Rotation_Alpha {type=number}
 Shotgun01 {type=boolean}
 Shotgun01_out {type=number}
 SimlationSpace {type=AnimPhysSimSpaceType}
 SkelMesh {type=SkeletalMeshComponent}
 SleevesRotation {type=Rotator}
 Velocity Interp Speed {type=number}
 WPO Pho Turbulence {type=number}
 WPO Pho Wind Intensity {type=number}
 ZiplineAlpha {type=number}
]
