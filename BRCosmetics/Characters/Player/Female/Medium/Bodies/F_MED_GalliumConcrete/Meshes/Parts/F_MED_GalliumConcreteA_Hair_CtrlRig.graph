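# Control Rig for the GalliumConcreteA female character's hair part, adjusting dynamic hair bones to correct their pose as the body moves.
# Head rotation relative to neck_01 is split into Euler deltas and remapped into weights for additive offsets on hair bones.
# Further fixes follow the MeleeAttack_out_deform curve, smoothed and zeroed when wearing a backpack, and Skydive_out_deform.
# A debug branch, off by default, drives head and clavicles from test controls.
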ControlRigBlueprint /BRCosmetics/Characters/Player/Female/Medium/Bodies/F_MED_GalliumConcrete/Meshes/Parts/F_MED_GalliumConcreteA_Hair_CtrlRig
graphs [
 New Function_ContainedGraph {nodes [
   n1 = RigVMFunctionEntryNode()
   RigVMFunction_Sequence = &RigVMFunction_Sequence(ExecuteContext=@n1.ExecuteContext)
   DISPATCH_RigVMDispatch_If = RigVMDispatchNode(True=0, False=RigVMVariableNode() {variable=MeleeAttack_out_deform}.Value, Condition=RigVMVariableNode() {variable=wearing Backpack}.Value) {template="DISPATCH_RigVMDispatch_If(in Condition,in True,in False,out Result)"}
   AlphaInterp = &RigVMFunction_AlphaInterp(InterpSpeedDecreasing=1, InterpSpeedIncreasing=8, Scale=1, bClampResult=false, bMapRange=false, Value=@DISPATCH_RigVMDispatch_If.Result)
   &RigUnit_ModifyTransforms(
    ItemToModify="((Item=(Type=Bone,Name=\"None\"),Transform=(Rotation=(X=0,Y=0,Z=0,W=1),Translation=(X=0,Y=0,Z=0),Scale3D=(X=1,Y=1,Z=1))),(Item=(Type=Bone,Name=\"dyn_hair_e_MSTR\"),Transform=(Rotation=(X=-0.0151,Y=-0.173,Z=-0.0858,W=0.9811),Translation=(X=0,Y=0,Z=0),Scale3D=(X ...(truncated)"
    WeightMinimum=0
    Weight=@AlphaInterp.Result
    ExecutePin=@RigVMFunction_Sequence.A
   )
   RigVMFunctionReturnNode(ExecuteContext=@RigVMFunction_Sequence.B)
  ]}
 New Function_ContainedGraph {nodes [
   n1 = RigVMFunctionEntryNode()
   RigVMFunction_ControlFlowBranch = &RigVMFunction_ControlFlowBranch(Condition=@n1.Debug, ExecuteContext=@n1.ExecuteContext)
   RigVMFunctionReturnNode(ExecuteContext=@RigVMFunction_ControlFlowBranch.Completed)
   RigVMFunction_Sequence = RigVMAggregateNode(ExecuteContext=@RigVMFunction_ControlFlowBranch.True)
   RigUnit_GetTransform = &RigUnit_GetTransform(Item=(Type=Control,Name="clavicle_l_ctrl"), bInitial=False)
   RigUnit_SetTransform = &RigUnit_SetTransform(Item=(Type=Bone,Name="clavicle_l"), Weight=1, bInitial=False, Value=@RigUnit_GetTransform.Transform, ExecutePin=@RigVMFunction_Sequence.A)
   &RigUnit_SetTransform(
    Item=(Type=Bone,Name="clavicle_r")
    Weight=1
    bInitial=False
    Value=&RigUnit_GetTransform(Item=(Type=Control,Name="clavicle_r_ctrl"), bInitial=False).Transform
    ExecutePin=@RigVMFunction_Sequence.B
   )
   &RigUnit_SetTransform(
    Item=(Type=Bone,Name="head")
    Weight=1
    bInitial=False
    Value=&RigUnit_GetTransform(Item=(Type=Control,Name="head_ctrl"), bInitial=False).Transform
    ExecutePin=@RigVMFunction_Sequence.C
   )
  ]}
 New Function_ContainedGraph {nodes [
   n1 = RigVMFunctionEntryNode()
   RigVMFunction_MathQuaternionToEuler = &RigVMFunction_MathQuaternionToEuler(
    RotationOrder=ZYX
    Value=&RigUnit_GetRelativeTransformForItem(
     Child=(Type=Bone,Name="head")
     Parent=(Type=Bone,Name="neck_01")
     bChildInitial=False
     bParentInitial=False
     Child.Name=@n1.Child
     Parent.Name=@n1.Parent
    ).RelativeTransform.Rotation
   )
   Remap = &RigVMFunction_MathDoubleRemap(
    TargetMaximum=1
    TargetMinimum=0
    SourceMaximum=@n1.Source Max
    SourceMinimum=@n1.Source Min
    Value=@RigVMFunction_MathQuaternionToEuler.Result.Z
   )
   > RigVMFunctionReturnNode(
    Delta_X_Negative=&RigVMFunction_MathDoubleRemap(SourceMaximum=-60, SourceMinimum=-1, TargetMaximum=1, TargetMinimum=0, Value=@RigVMFunction_MathQuaternionToEuler.Result.X).Result
    Delta_X_Positive=&RigVMFunction_MathDoubleRemap(SourceMaximum=60, SourceMinimum=0, TargetMaximum=1, TargetMinimum=0, Value=@RigVMFunction_MathQuaternionToEuler.Result.X).Result
    Delta_z=@Remap.Result
   )
  ]}
 New Function_ContainedGraph {nodes [
   n1 = RigVMFunctionEntryNode()
   RigVMFunction_Sequence = &RigVMFunction_Sequence(ExecuteContext=@n1.ExecuteContext)
   &RigUnit_ModifyTransforms(
    ItemToModify="((Item=(Type=Bone,Name=\"dyn_hair_f_MSTR\"),Transform=(Rotation=(X=0,Y=0,Z=0,W=1),Translation=(X=0,Y=0,Z=1.5),Scale3D=(X=1,Y=1,Z=1))),(Item=(Type=Bone,Name=\"dyn_hair_e_MSTR\"),Transform=(Rotation=(X=0,Y=0,Z=0,W=1),Translation=(X=0,Y=0,Z=1.5),Scale3D=(X=1,Y=1, ...(truncated)"
    WeightMinimum=0
    ExecutePin=@RigVMFunction_Sequence.A
    Weight=RigVMVariableNode() {variable=Skydive_out_deform}.Value
   )
   RigVMFunctionReturnNode(ExecuteContext=@RigVMFunction_Sequence.B)
  ]}
 RigVMFunctionLibrary {nodes [
   Controls = RigVMCollapseNode(Debug=False, ExecuteContext=())
   DeltaRotation = RigVMCollapseNode(ExecuteContext=())
   Meleeattack_Fix = RigVMCollapseNode(ExecuteContext=())
   Skydive_fix = RigVMCollapseNode(ExecuteContext=())
  ]}
 RigVMFunction_Sequence_1_ContainedGraph {nodes [
   n1 = RigVMFunctionEntryNode()
   RigVMFunction_Sequence = &RigVMFunction_Sequence(ExecuteContext=@n1.ExecuteContext)
   n2 = &RigVMFunction_Sequence(ExecuteContext=@RigVMFunction_Sequence.B)
   RigVMFunctionReturnNode(A=@RigVMFunction_Sequence.A, B=@n2.A, C=@n2.B)
  ]}
 RigVMModel {nodes [
   RigUnit_BeginExecution = &RigUnit_BeginExecution()
   Controls = RigVMFunctionReferenceNode(Debug=false, ExecuteContext=@RigUnit_BeginExecution.ExecutePin)
   > DeltaRotation = RigVMFunctionReferenceNode(Child=Head, Parent=neck_01, Source Max=-60, Source Min=-10)
   > &RigUnit_ModifyTransforms(
    ItemToModify="((Item=(Type=Bone,Name=\"dyn_hair_e_MSTR\"),Transform=(Rotation=(X=0,Y=-0.5736,Z=0,W=0.8192),Translation=(X=0,Y=0,Z=0),Scale3D=(X=1,Y=1,Z=1))),(Item=(Type=Bone,Name=\"dyn_hair_d_MSTR\"),Transform=(Rotation=(X=0,Y=-0.5736,Z=0,W=0.8192),Translation=(X=0,Y=0,Z=0) ...(truncated)"
    WeightMinimum=0
    Weight=@DeltaRotation.Delta_z
   )
   > &RigUnit_ModifyTransforms(
    ItemToModify="((Item=(Type=Bone,Name=\"dyn_hair_e_MSTR\"),Transform=(Rotation=(X=-0.2113,Y=-0.4532,Z=-0.366,W=0.7849),Translation=(X=0,Y=0,Z=0),Scale3D=(X=1,Y=1,Z=1))),(Item=(Type=Bone,Name=\"dyn_hair_d_MSTR\"),Transform=(Rotation=(X=0,Y=0,Z=0.1736,W=0.9848),Translation=(X= ...(truncated)"
    WeightMinimum=0
    Weight=@DeltaRotation.Delta_X_Positive
   )
   > &RigUnit_ModifyTransforms(
    ItemToModify="((Item=(Type=Bone,Name=\"dyn_hair_e_MSTR\"),Transform=(Rotation=(X=0.1094,Y=-0.2346,Z=0.4082,W=0.8754),Translation=(X=0,Y=0,Z=0),Scale3D=(X=1,Y=1,Z=1))),(Item=(Type=Bone,Name=\"dyn_hair_d_MSTR\"),Transform=(Rotation=(X=0.1094,Y=-0.2346,Z=0.4082,W=0.8754),Trans ...(truncated)"
    WeightMinimum=0
    Weight=@DeltaRotation.Delta_X_Negative
   )
   > Meleeattack_Fix = RigVMFunctionReferenceNode()
   > Skydive_fix = RigVMFunctionReferenceNode()
  ]}
]
hierarchy [
 root {loc=<0 0 0>, rot=<0 0 0 1>}
 pelvis {parent=root, loc=<0 -0.5301 98.0571>, rot=<0.0127 0.707 -0.0127 -0.707>}
 spine_01 {parent=pelvis, loc=<6.1208 -0.9328 0>, rot=<0 0 0.034 -0.9994>}
 spine_02 {parent=spine_01, loc=<6.5996 -0.2095 0>, rot=<0 0 0.0238 -0.9997>}
 spine_03 {parent=spine_02, loc=<6.4914 -0.1237 0>, rot=<0 0 0.0149 -0.9999>}
 spine_04 {parent=spine_03, loc=<6.9598 0.0794 0>, rot=<0 0 -0.0132 -0.9999>}
 spine_05 {parent=spine_04, loc=<6.9987 0 0>, rot=<0 0 -0.0015 -1>}
 clavicle_l {parent=spine_05, loc=<5.4831 -0.1053 -3.2946>, rot=<0.0465 -0.7289 -0.009 -0.6829>}
 upperarm_l {parent=clavicle_l, loc=<9.8658 -1.2136 -0.4105>, rot=<0.0164 -0.3245 -0.0269 -0.9454>}
 lowerarm_l {parent=upperarm_l, loc=<26.465 -0.8345 0.259>, rot=<0.0036 0.0161 -0.2298 -0.9731>}
 hand_l {parent=lowerarm_l, loc=<22.7355 -0.3203 -0.6105>, rot=<-0.6784 -0.0223 0.0735 -0.7307>}
 lowerarm_twist_01_l {parent=lowerarm_l, loc=<18.316 0 0>, rot=<0 0 0 1>}
 lowerarm_twist_02_l {parent=lowerarm_l, loc=<10.6707 0 0>, rot=<0 0 0 1>}
 upperarm_twist_01_l {parent=upperarm_l, loc=<7.2135 -0.1134 0>, rot=<0 0 0 1>}
 upperarm_twist_02_l {parent=upperarm_l, loc=<19.7998 -0.1134 0>, rot=<0 0 0 1>}
 clavicle_r {parent=spine_05, loc=<5.4831 -0.1053 3.2946>, rot=<0.7289 0.0465 -0.6829 0.009>}
 upperarm_r {parent=clavicle_r, loc=<-9.8658 1.2136 0.4104>, rot=<0.0164 -0.3245 -0.0269 -0.9454>}
 lowerarm_r {parent=upperarm_r, loc=<-26.4651 0.8345 -0.259>, rot=<0.0036 0.0161 -0.2298 -0.9731>}
 hand_r {parent=lowerarm_r, loc=<-22.7356 0.3203 0.6106>, rot=<-0.6784 -0.0223 0.0735 -0.7307>}
 lowerarm_twist_01_r {parent=lowerarm_r, loc=<-18.3161 0 0>, rot=<0 0 0 1>}
 lowerarm_twist_02_r {parent=lowerarm_r, loc=<-10.6707 0 0>, rot=<0 0 0 1>}
 upperarm_twist_01_r {parent=upperarm_r, loc=<-7.2135 0.1134 0>, rot=<0 0 0 1>}
 upperarm_twist_02_r {parent=upperarm_r, loc=<-19.7998 0.1134 0>, rot=<0 0 0 1>}
 neck_01 {parent=spine_05, loc=<11.3138 0.1529 0>, rot=<0 0 -0.2003 -0.9797>}
 neck_02 {parent=neck_01, loc=<4.5627 0 0>, rot=<0 0 0.0504 -0.9987>}
 head {parent=neck_02, loc=<5.3914 0 0>, rot=<0 0 0.115 -0.9934>}
 dyn_bangs_c_01 {parent=head, loc=<8.5955 6.5041 -4.9884>, rot=<0.0878 -0.9429 -0.2831 -0.1519>}
 dyn_hair_a_MSTR {parent=head, loc=<1.7547 -3.2987 -6.6603>, rot=<-0.0738 -0.995 -0.0117 -0.0658>}
 dyn_hair_a_01 {parent=dyn_hair_a_MSTR, loc=<0 0 0>, rot=<0 0 0 1>}
 dyn_hair_a_02 {parent=dyn_hair_a_01, loc=<4 0 0>, rot=<0 0 0 1>}
 dyn_hair_a_03 {parent=dyn_hair_a_02, loc=<4 0 0>, rot=<0 0 -0.0619 -0.9981>}
 dyn_hair_a_04 {parent=dyn_hair_a_03, loc=<4 0 0>, rot=<0.0034 -0.0161 -0.2056 -0.9785>}
 dyn_hair_a_05 {parent=dyn_hair_a_04, loc=<3.366 0 0>, rot=<0 0 0 1>}
 dyn_hair_b_01 {parent=dyn_hair_a_02, loc=<4.1405 1.8534 -1.6007>, rot=<0.1561 -0.0574 -0.3335 -0.928>}
 dyn_hair_c_MSTR {parent=head, loc=<-0.3901 -4.7659 -5.6499>, rot=<0.0346 -0.96 0.2378 -0.1436>}
 dyn_hair_c_01 {parent=dyn_hair_c_MSTR, loc=<0 0 0>, rot=<0 0 0 1>}
 dyn_hair_c_02 {parent=dyn_hair_c_01, loc=<4.7395 0 0>, rot=<0 0 0 1>}
 dyn_hair_c_03 {parent=dyn_hair_c_02, loc=<4.7395 0 0>, rot=<0 0 0 1>}
 dyn_hair_d_MSTR {parent=head, loc=<-0.2933 -6.0824 -3.6439>, rot=<0.1014 -0.8214 0.535 -0.1697>}
 dyn_hair_d_01 {parent=dyn_hair_d_MSTR, loc=<0 0 0>, rot=<0 0 0 1>}
 dyn_hair_d_02 {parent=dyn_hair_d_01, loc=<4.2339 0 -0.0136>, rot=<0 0.0063 0 -1>}
 dyn_hair_d_03 {parent=dyn_hair_d_02, loc=<4.2339 0 -0.0136>, rot=<0 -0.0234 0 -0.9997>}
 dyn_hair_d_04 {parent=dyn_hair_d_03, loc=<4.2339 0 -0.0136>, rot=<0 -0.1381 0 -0.9904>}
 dyn_hair_d_05 {parent=dyn_hair_d_04, loc=<4.2343 0 0>, rot=<0 0 0 1>}
 dyn_hair_e_MSTR {parent=head, loc=<-0.2877 -6.6529 0.0597>, rot=<0.1513 -0.6895 0.6943 -0.1398>}
 dyn_hair_e_01 {parent=dyn_hair_e_MSTR, loc=<0 0 0>, rot=<0 0 0 1>}
 dyn_hair_e_02 {parent=dyn_hair_e_01, loc=<4.2339 0 -0.0136>, rot=<0 0.0063 0 -1>}
 dyn_hair_e_03 {parent=dyn_hair_e_02, loc=<4.2339 0 -0.0136>, rot=<0 -0.0234 0 -0.9997>}
 dyn_hair_e_04 {parent=dyn_hair_e_03, loc=<4.2339 0 -0.0136>, rot=<0 -0.1381 0 -0.9904>}
 dyn_hair_e_05 {parent=dyn_hair_e_04, loc=<4.2343 0 0>, rot=<0 0 0 1>}
 dyn_hair_f_MSTR {parent=head, loc=<-0.2137 -5.7643 4.2854>, rot=<0.1803 -0.6008 0.7691 -0.1223>}
 dyn_hair_f_01 {parent=dyn_hair_f_MSTR, loc=<0 0 0>, rot=<0 0 0 1>}
 dyn_hair_f_02 {parent=dyn_hair_f_01, loc=<4.2339 0 -0.0136>, rot=<0 -0.0289 0 -0.9996>}
 dyn_hair_f_03 {parent=dyn_hair_f_02, loc=<4.2339 0 -0.0136>, rot=<0 -0.0289 0 -0.9996>}
 dyn_hair_f_04 {parent=dyn_hair_f_03, loc=<4.2339 0 -0.0136>, rot=<0 -0.0289 0 -0.9996>}
 dyn_hair_f_05 {parent=dyn_hair_f_04, loc=<4.2343 0 0>, rot=<0 0 0 1>}
 dyn_hair_g_MSTR {parent=head, loc=<1.5094 -3.1392 6.9169>, rot=<0.0934 -0.2067 0.9736 -0.0256>}
 dyn_hair_g_01 {parent=dyn_hair_g_MSTR, loc=<0 0 0>, rot=<0 0 0 1>}
 dyn_hair_g_02 {parent=dyn_hair_g_01, loc=<4.7395 0 0>, rot=<0 0 0 1>}
 dyn_hair_g_03 {parent=dyn_hair_g_02, loc=<4.7395 0 0>, rot=<0 0 0 1>}
 dyn_hair_h_MSTR {parent=head, loc=<3.9354 0.084 7.6367>, rot=<0.0355 0.0316 0.9988 -0.0083>}
 dyn_hair_h_01 {parent=dyn_hair_h_MSTR, loc=<0 0 0>, rot=<0 0 0 1>}
 dyn_hair_h_02 {parent=dyn_hair_h_01, loc=<4.7546 -0.008 -0.0537>, rot=<0 0 0 1>}
 dyn_hair_h_03 {parent=dyn_hair_h_02, loc=<4.7546 -0.008 -0.0537>, rot=<0 0 0 1>}
 dyn_hair_h_04 {parent=dyn_hair_h_03, loc=<4.7546 -0.008 -0.0537>, rot=<0 0 0.3848 -0.923>}
 dyn_hair_i_MSTR {parent=head, loc=<5.0399 4.8346 8.0229>, rot=<-0.1464 0.2425 0.959 0.0043>}
 dyn_hair_i_01 {parent=dyn_hair_i_MSTR, loc=<0 0 0>, rot=<0 0 0 1>}
 dyn_hair_i_02 {parent=dyn_hair_i_01, loc=<3.366 0 0>, rot=<0 0.1468 0 -0.9892>}
 dyn_bangs_a_MSTR {parent=head, loc=<9.8559 7.9526 2.5838>, rot=<0.2604 0.5267 0.8077 -0.049>}
 dyn_bangs_a_01 {parent=dyn_bangs_a_MSTR, loc=<0 0 0>, rot=<0 0 0 1>}
 dyn_bangs_a_02 {parent=dyn_bangs_a_01, loc=<3.366 0 0>, rot=<0 -0.1876 0 -0.9822>}
 dyn_bangs_b_MSTR {parent=head, loc=<10.097 8.2362 -2.1592>, rot=<0.2666 0.6686 0.6847 -0.1143>}
 dyn_bangs_b_01 {parent=dyn_bangs_b_MSTR, loc=<0 0 0>, rot=<0 0 0 1>}
 dyn_bangs_b_02 {parent=dyn_bangs_b_01, loc=<2.4109 0 0.3788>, rot=<0 -0.1876 0 -0.9822>}
 dyn_hat_MSTR {parent=head, loc=<14.8215 -4.2416 -0.7229>, rot=<0.4104 -0.5655 0.508 -0.5038>}
 dyn_hat_01 {parent=dyn_hat_MSTR, loc=<0 0 0>, rot=<0 0.1106 0 -0.9939>}
 dyn_hat_02 {parent=dyn_hat_01, loc=<3.366 0 0>, rot=<0 -0.1876 0 -0.9822>}
 dyn_glasses {parent=head, loc=<3.29 8.8969 -0.0231>, rot=<-0.5 -0.5 -0.5 -0.5>}
 clavicle_l_ctrl {controlType=EULER_TRANSFORM, type=Control, loc=<3.2946 -4.1138 136.616>, rot=<0.0008 0.0336 -0.0252 0.9991>}
 clavicle_r_ctrl {controlType=EULER_TRANSFORM, type=Control, loc=<-3.2946 -4.1138 136.6161>, rot=<0.9991 0.0252 0.0336 -0.0008>}
 head_ctrl {controlType=EULER_TRANSFORM, type=Control, loc=<0 -1.6873 152.0342>, rot=<0.0029 -0.7071 -0.0029 0.7071>}
]
variables [
 MeleeAttack_out_deform {type=double, default=0}
 Skydive_out_deform {type=double, default=0}
 wearing Backpack {type=bool, default=False}
]
