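# Physics asset for the hair of the female medium GalliumConcrete player character, defining capsule bodies for dynamic hair bones and the upper body.
# Lettered hair chains hang from MSTR bodies on the head, except strand b, which branches off chain a.
# Swing limits generally widen toward the tips, with zero twist.
# Capsules on spine, clavicles, neck and head form the body volume, linked by free-swing constraints, as is a hat master body.
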PhysicsAsset /BRCosmetics/Characters/Player/Female/Medium/Bodies/F_MED_GalliumConcrete/Meshes/Parts/F_MED_GalliumConcreteA_Hair_Physics
bodies {
 clavicle_l [
  {shapeType=Capsule, length=13.3647, radius=6.2743, center=<4.8891 -0.6882 -1.651>, rotation=<rot 90 0 -8.4719>}
  {shapeType=Capsule, length=21, radius=9, shapeName=clavicle_l_capsule, center=<24.5589 0 -14.8923>, rotation=<rot 50 -0.0003 -0.0003>}
  {shapeType=Capsule, length=10, radius=1.5, shapeName=clavicle_l_capsule, center=<5.0854 -1.9946 5.0194>, rotation=<rot -90 180 177.1058>}
 ]
 clavicle_r [{shapeType=Capsule, length=25.5088, radius=5.7155, center=<-14.538 0.3005 1.1333>, rotation=<rot 0.0563 -86.3476 91.7762>}]
 dyn_hair_a_01 [{shapeType=Capsule, length=1, radius=1.5, shapeName=dyn_hair_a_01_capsule, center=<2.0826 0 0>, rotation=<rot 90 0 0>}]
 dyn_hair_a_02 [{shapeType=Capsule, length=1, radius=1.5, shapeName=dyn_hair_e_01_capsule, center=<2.0504 0 0>, rotation=<rot 90 0 0>}]
 dyn_hair_a_03 [
  {shapeType=Capsule, length=2.7316, radius=2.5, shapeName=dyn_hair_a_03_capsule, center=<2.0741 0 0>, rotation=<rot 90 0 0>}
  {
   shapeType=Capsule
   length=3.5327
   radius=1.0196
   shapeName=dyn_hair_a_03_capsule
   center=<3.1389 -1.9839 1.4246>
   rotation=<rot 67.0572 151.4688 -180>
  }
 ]
 dyn_hair_a_04 [{shapeType=Capsule, length=2.7316, radius=1, shapeName=dyn_hair_a_04_capsule, center=<2.0741 0 0>, rotation=<rot 90 0 0>}]
 dyn_hair_a_05 [{shapeType=Capsule, length=1, radius=0.7, shapeName=dyn_hair_a_05_capsule, center=<1.6354 0 0>, rotation=<rot 90 0 0>}]
 dyn_hair_a_MSTR [{shapeType=Capsule, length=0.505, radius=0.505}]
 dyn_hair_b_01 [{shapeType=Capsule, length=5, radius=1.4788, shapeName=dyn_hair_b_01_capsule, center=<3.3714 0 -0.7294>, rotation=<rot 80.0001 0.0003 0.0001>}]
 dyn_hair_c_01 [{shapeType=Capsule, length=1, radius=1.8, shapeName=dyn_hair_c_01_capsule, center=<1 0 0.8946>, rotation=<rot 90 0 0>}]
 dyn_hair_c_02 [{shapeType=Capsule, length=2.5, radius=1.5, shapeName=dyn_hair_c_02_capsule, center=<1 0 0.8946>, rotation=<rot 90 0 0>}]
 dyn_hair_c_03 [{
   shapeType=Capsule
   length=4.1472
   radius=1.5
   shapeName=dyn_hair_c_03_capsule
   center=<1.6532 -1.9392 1.7511>
   rotation=<rot 51.5546 -120.7424 -100.7756>
  }]
 dyn_hair_c_MSTR [{shapeType=Capsule, length=0.505, radius=0.505}]
 dyn_hair_d_01 [{shapeType=Capsule, length=2.5, radius=1.5, shapeName=dyn_hair_d_01_capsule, rotation=<rot 90 0 0>}]
 dyn_hair_d_02 [{shapeType=Capsule, length=1, radius=1.8, shapeName=dyn_hair_d_02_capsule, center=<1 0 0>, rotation=<rot 90 0 0>}]
 dyn_hair_d_03 [{shapeType=Capsule, length=1, radius=1.8, shapeName=dyn_hair_d_03_capsule, center=<1 0 0>, rotation=<rot 90 0 0>}]
 dyn_hair_d_04 [{shapeType=Capsule, length=1, radius=1.8, shapeName=dyn_hair_d_04_capsule, center=<1 0 0>, rotation=<rot 90 0 0>}]
 dyn_hair_d_05 [{
   shapeType=Capsule
   length=1.6969
   radius=1.5
   shapeName=dyn_hair_d_05_capsule
   center=<1.8013 -1 -0.5>
   rotation=<rot 65.4358 -22.5191 0.0001>
  }]
 dyn_hair_d_MSTR [{shapeType=Capsule, length=0.505, radius=0.505}]
 dyn_hair_e_01 [{shapeType=Capsule, length=1, radius=1.8, shapeName=dyn_hair_e_01_capsule, center=<1 0 0>, rotation=<rot 90 0 0>}]
 dyn_hair_e_02 [{shapeType=Capsule, length=1, radius=1.8, shapeName=dyn_hair_e_02_capsule, center=<1 0 0>, rotation=<rot 90 0 0>}]
 dyn_hair_e_03 [{shapeType=Capsule, length=1, radius=1.8, shapeName=dyn_hair_e_03_capsule, center=<1 0 0>, rotation=<rot 90 0 0>}]
 dyn_hair_e_04 [{shapeType=Capsule, length=1, radius=1.8, shapeName=dyn_hair_e_04_capsule, center=<1 0 0>, rotation=<rot 90 0 0>}]
 dyn_hair_e_05 [{shapeType=Capsule, length=5, radius=2.2, shapeName=dyn_hair_e_05_capsule, center=<4.177 0 -1>, rotation=<rot 77.4886 0 0>}]
 dyn_hair_e_MSTR [{shapeType=Capsule, length=0.505, radius=0.505}]
 dyn_hair_f_01 [{shapeType=Capsule, length=1, radius=1.8, shapeName=dyn_hair_f_01_capsule, center=<1 0 0>, rotation=<rot 90 0 0>}]
 dyn_hair_f_02 [
  {shapeType=Capsule, length=1, radius=1.8, shapeName=dyn_hair_f_02_capsule, center=<1 0 1.6725>, rotation=<rot 90 0 0>}
  {shapeType=Capsule, length=3.3625, radius=1.5964, shapeName=dyn_hair_f_02_capsule, center=<2.9455 -3.5 -1.1>, rotation=<rot 90 0 0>}
 ]
 dyn_hair_f_03 [{shapeType=Capsule, length=1, radius=1.8, shapeName=dyn_hair_f_03_capsule, center=<1 0 0>, rotation=<rot 90 0 0>}]
 dyn_hair_f_04 [
  {shapeType=Capsule, length=1, radius=1.8, shapeName=dyn_hair_f_04_capsule, center=<1 0 0>, rotation=<rot 90 0 0>}
  {
   shapeType=Capsule
   length=3.2129
   radius=1
   shapeName=dyn_hair_f_04_capsule
   center=<1.3152 -4.1157 -0.2673>
   rotation=<rot 77.3287 179.9999 172.0566>
  }
 ]
 dyn_hair_f_05 [{shapeType=Capsule, length=4, radius=1.7, shapeName=dyn_hair_f_05_capsule, center=<3.4024 0 -0.7112>, rotation=<rot 63.1084 0 0>}]
 dyn_hair_f_MSTR [{shapeType=Capsule, length=0.505, radius=0.505}]
 dyn_hair_g_01 [{shapeType=Capsule, length=1.5, radius=1.5, shapeName=dyn_hair_g_01_capsule, center=<1 0 0>, rotation=<rot 90 0 0>}]
 dyn_hair_g_02 [{shapeType=Capsule, length=1.5, radius=1.5, shapeName=dyn_hair_g_02_capsule, center=<1 0 0>, rotation=<rot 90 0 0>}]
 dyn_hair_g_03 [{shapeType=Capsule, length=1.5, radius=1.5, shapeName=dyn_hair_g_03_capsule, center=<1 0 0>, rotation=<rot 90 0 0>}]
 dyn_hair_g_MSTR [{shapeType=Capsule, length=0.505, radius=0.505}]
 dyn_hair_h_01 [{shapeType=Capsule, length=2, radius=1, shapeName=dyn_hair_h_01_capsule, center=<2 0 0>, rotation=<rot 90 0 0>}]
 dyn_hair_h_02 [{shapeType=Capsule, length=2.2007, radius=1.27, shapeName=dyn_hair_h_02_capsule, center=<2.2048 0 -0.0003>, rotation=<rot 0 0 90>}]
 dyn_hair_h_03 [{shapeType=Capsule, length=2.8644, radius=1.27, shapeName=dyn_hair_h_03_capsule, center=<2.2048 0 0>, rotation=<rot 90 0 0>}]
 dyn_hair_h_04 [{shapeType=Capsule, length=4.4103, radius=1.2579, center=<3.2655 -0.0165 0.3553>, rotation=<rot 76.7288 -11.4548 0>}]
 dyn_hair_h_MSTR [{shapeType=Capsule, length=0.505, radius=0.505}]
 dyn_hair_i_01 [{shapeType=Capsule, length=1, radius=1, shapeName=dyn_hair_i_01_capsule, center=<2 0 0>, rotation=<rot 90 0 0>}]
 dyn_hair_i_02 [{shapeType=Capsule, length=2, radius=1, shapeName=dyn_hair_i_02_capsule, center=<2 0 0>, rotation=<rot 90 0 0>}]
 dyn_hair_i_MSTR [{shapeType=Capsule, length=0.505, radius=0.505}]
 dyn_hat_MSTR [{shapeType=Capsule, length=0.505, radius=0.505}]
 head [
  {shapeType=Capsule, length=5.5862, radius=5.9015, center=<0.8118 3.4354 0>, rotation=<rot 90 0 0>}
  {shapeType=Capsule, length=6.7157, radius=4.772, center=<0.8118 1.1771 0>, rotation=<rot 90 0 0>}
  {shapeType=Capsule, length=9.347, radius=4.522, center=<1.8068 -1.0039 -0.1194>, rotation=<rot 90 0 -10.0004>}
 ]
 neck_01 [
  {shapeType=Capsule, length=3.7617, radius=5}
  {shapeType=Capsule, length=56, radius=1, shapeName=neck_01_capsule, center=<-1.9741 -2.9586 0>}
  {shapeType=Capsule, length=18.5, radius=2, shapeName=neck_01_capsule, center=<2.4495 -1.3786 11.6368>}
 ]
 neck_02 [{shapeType=Capsule, length=11.8, radius=15, shapeName=neck_02_capsule, center=<-4.9981 8.6354 0>}]
 spine_04 [{shapeType=Capsule, length=14.2834, radius=10.0962, center=<0 -0.7495 0>}]
 spine_05 [{shapeType=Capsule, length=23.7679, radius=8.5, center=<0 -1.2704 0>}]
}
constraints [
 {bone1Name=dyn_hair_a_01, bone2Name=dyn_hair_a_MSTR, swing1LimitDegrees=7.5, swing2LimitDegrees=7.5, twistLimitDegrees=0}
 {bone1Name=dyn_hair_a_02, bone2Name=dyn_hair_a_01, swing1LimitDegrees=5, swing2LimitDegrees=5, twistLimitDegrees=0}
 {bone1Name=dyn_hair_a_03, bone2Name=dyn_hair_a_02, swing1LimitDegrees=8, swing2LimitDegrees=8, twistLimitDegrees=0}
 {bone1Name=dyn_hair_a_04, bone2Name=dyn_hair_a_03, swing1LimitDegrees=10, swing2LimitDegrees=10, twistLimitDegrees=0}
 {bone1Name=dyn_hair_a_05, bone2Name=dyn_hair_a_04, swing1LimitDegrees=15, swing2LimitDegrees=15, twistLimitDegrees=0}
 {bone1Name=dyn_hair_b_01, bone2Name=dyn_hair_a_02, swing1LimitDegrees=10, swing2LimitDegrees=10, twistLimitDegrees=0}
 {bone1Name=dyn_hair_c_01, bone2Name=dyn_hair_c_MSTR, swing1LimitDegrees=10, swing2LimitDegrees=15, twistLimitDegrees=0}
 {bone1Name=dyn_hair_c_02, bone2Name=dyn_hair_c_01, swing1LimitDegrees=20, swing2LimitDegrees=15, twistLimitDegrees=0}
 {bone1Name=dyn_hair_c_03, bone2Name=dyn_hair_c_02, swing1LimitDegrees=20, swing2LimitDegrees=25, twistLimitDegrees=0}
 {bone1Name=dyn_hair_d_01, bone2Name=dyn_hair_d_MSTR, swing1LimitDegrees=7.5, swing2LimitDegrees=45, twistLimitDegrees=0}
 {bone1Name=dyn_hair_d_02, bone2Name=dyn_hair_d_01, swing1LimitDegrees=7.5, swing2LimitDegrees=10, twistLimitDegrees=0}
 {bone1Name=dyn_hair_d_03, bone2Name=dyn_hair_d_02, swing1LimitDegrees=7.5, swing2LimitDegrees=15, twistLimitDegrees=0}
 {bone1Name=dyn_hair_d_04, bone2Name=dyn_hair_d_03, swing1LimitDegrees=10, swing2LimitDegrees=25, twistLimitDegrees=0}
 {bone1Name=dyn_hair_d_05, bone2Name=dyn_hair_d_04, swing1LimitDegrees=10, swing2LimitDegrees=25, twistLimitDegrees=0}
 {bone1Name=dyn_hair_e_01, bone2Name=dyn_hair_e_MSTR, swing1LimitDegrees=7, swing2LimitDegrees=7, twistLimitDegrees=0}
 {bone1Name=dyn_hair_e_02, bone2Name=dyn_hair_e_01, swing1LimitDegrees=10, swing2LimitDegrees=15, twistLimitDegrees=0}
 {bone1Name=dyn_hair_e_03, bone2Name=dyn_hair_e_02, swing1LimitDegrees=10, swing2LimitDegrees=15, twistLimitDegrees=0}
 {bone1Name=dyn_hair_e_04, bone2Name=dyn_hair_e_03, swing1LimitDegrees=15, swing2LimitDegrees=20, twistLimitDegrees=0}
 {bone1Name=dyn_hair_e_05, bone2Name=dyn_hair_e_04, swing1LimitDegrees=15, swing2LimitDegrees=20, twistLimitDegrees=0}
 {bone1Name=dyn_hair_f_01, bone2Name=dyn_hair_f_MSTR, swing1LimitDegrees=7, swing2LimitDegrees=7, twistLimitDegrees=0}
 {bone1Name=dyn_hair_g_01, bone2Name=dyn_hair_g_MSTR, swing1LimitDegrees=5, swing2LimitDegrees=15, twistLimitDegrees=0}
 {bone1Name=dyn_hair_g_02, bone2Name=dyn_hair_g_01, swing1LimitDegrees=7, swing2LimitDegrees=14, twistLimitDegrees=0}
 {bone1Name=dyn_hair_g_03, bone2Name=dyn_hair_g_02, swing1LimitDegrees=10, swing2LimitDegrees=15, twistLimitDegrees=0}
 {bone1Name=dyn_hair_h_01, bone2Name=dyn_hair_h_MSTR, swing1LimitDegrees=5, swing2LimitDegrees=5, twistLimitDegrees=0}
 {bone1Name=dyn_hair_h_02, bone2Name=dyn_hair_h_01, swing1LimitDegrees=7, swing2LimitDegrees=5, twistLimitDegrees=0}
 {bone1Name=dyn_hair_h_03, bone2Name=dyn_hair_h_02, swing1LimitDegrees=8, swing2LimitDegrees=8, twistLimitDegrees=0}
 {bone1Name=dyn_hair_h_04, bone2Name=dyn_hair_h_03, swing1LimitDegrees=10, swing2LimitDegrees=10, twistLimitDegrees=0}
 {bone1Name=dyn_hair_i_01, bone2Name=dyn_hair_i_MSTR, swing1LimitDegrees=5, swing2LimitDegrees=5, twistLimitDegrees=0}
 {bone1Name=dyn_hair_i_02, bone2Name=dyn_hair_i_01, swing1LimitDegrees=5, swing2LimitDegrees=5, twistLimitDegrees=0}
 {
  bone1Name=clavicle_l
  bone2Name=spine_05
  swing1LimitDegrees=45
  swing1Motion=Free
  swing2LimitDegrees=45
  swing2Motion=Free
  twistLimitDegrees=0
 }
 {
  bone1Name=clavicle_r
  bone2Name=spine_05
  swing1LimitDegrees=45
  swing1Motion=Free
  swing2LimitDegrees=45
  swing2Motion=Free
  twistLimitDegrees=0
 }
 {bone1Name=head, bone2Name=neck_02, swing1LimitDegrees=45, swing1Motion=Free, swing2LimitDegrees=45, swing2Motion=Free, twistLimitDegrees=0}
 {
  bone1Name=spine_05
  bone2Name=spine_04
  swing1LimitDegrees=45
  swing1Motion=Free
  swing2LimitDegrees=45
  swing2Motion=Free
  twistLimitDegrees=0
 }
 {
  bone1Name=dyn_hair_a_MSTR
  bone2Name=head
  swing1LimitDegrees=45
  swing1Motion=Free
  swing2LimitDegrees=45
  swing2Motion=Free
  twistLimitDegrees=0
 }
 {
  bone1Name=dyn_hair_c_MSTR
  bone2Name=head
  swing1LimitDegrees=45
  swing1Motion=Free
  swing2LimitDegrees=45
  swing2Motion=Free
  twistLimitDegrees=0
 }
 {
  bone1Name=dyn_hair_d_MSTR
  bone2Name=head
  swing1LimitDegrees=45
  swing1Motion=Free
  swing2LimitDegrees=45
  swing2Motion=Free
  twistLimitDegrees=0
 }
 {
  bone1Name=dyn_hair_e_MSTR
  bone2Name=head
  swing1LimitDegrees=45
  swing1Motion=Free
  swing2LimitDegrees=45
  swing2Motion=Free
  twistLimitDegrees=0
 }
 {
  bone1Name=dyn_hair_f_MSTR
  bone2Name=head
  swing1LimitDegrees=45
  swing1Motion=Free
  swing2LimitDegrees=45
  swing2Motion=Free
  twistLimitDegrees=0
 }
 {
  bone1Name=dyn_hair_g_MSTR
  bone2Name=head
  swing1LimitDegrees=45
  swing1Motion=Free
  swing2LimitDegrees=45
  swing2Motion=Free
  twistLimitDegrees=0
 }
 {
  bone1Name=dyn_hair_h_MSTR
  bone2Name=head
  swing1LimitDegrees=45
  swing1Motion=Free
  swing2LimitDegrees=45
  swing2Motion=Free
  twistLimitDegrees=0
 }
 {
  bone1Name=dyn_hair_i_MSTR
  bone2Name=head
  swing1LimitDegrees=45
  swing1Motion=Free
  swing2LimitDegrees=45
  swing2Motion=Free
  twistLimitDegrees=0
 }
 {
  bone1Name=dyn_hat_MSTR
  bone2Name=head
  swing1LimitDegrees=45
  swing1Motion=Free
  swing2LimitDegrees=45
  swing2Motion=Free
  twistLimitDegrees=0
 }
 {
  bone1Name=neck_01
  bone2Name=spine_05
  swing1LimitDegrees=45
  swing1Motion=Free
  swing2LimitDegrees=45
  swing2Motion=Free
  twistLimitDegrees=0
 }
 {bone1Name=dyn_hair_f_02, bone2Name=dyn_hair_f_01, swing1LimitDegrees=10, swing2LimitDegrees=15, twistLimitDegrees=0}
 {bone1Name=dyn_hair_f_03, bone2Name=dyn_hair_f_02, swing1LimitDegrees=15, swing2LimitDegrees=15, twistLimitDegrees=0}
 {bone1Name=dyn_hair_f_04, bone2Name=dyn_hair_f_03, swing1LimitDegrees=20, swing2LimitDegrees=20, twistLimitDegrees=0}
 {bone1Name=dyn_hair_f_05, bone2Name=dyn_hair_f_04, swing1LimitDegrees=30, swing2LimitDegrees=25, twistLimitDegrees=0}
 {
  bone1Name=neck_02
  bone2Name=neck_01
  swing1LimitDegrees=45
  swing1Motion=Free
  swing2LimitDegrees=45
  swing2Motion=Free
  twistLimitDegrees=45
  twistMotion=Free
 }
]
