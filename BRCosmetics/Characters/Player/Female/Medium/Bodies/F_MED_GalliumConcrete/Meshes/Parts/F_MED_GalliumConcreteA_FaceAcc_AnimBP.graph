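# Post-process style Anim Blueprint for GalliumConcrete's face accessory part, animating hair, hat, headband and bangs.
# The copied parent pose runs through a Control Rig, then AnimDynamics, then RigidBody hair simulation with front-end and gameplay variants.
# With back bling, a cape or rigid bodies disabled, output switches to simulated hair constrained toward clavicles, neck and spine.
# Event Graph updates gravity override, head tilt, melee, skydive and back bling inputs.
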AnimBlueprint /BRCosmetics/Characters/Player/Female/Medium/Bodies/F_MED_GalliumConcrete/Meshes/Parts/F_MED_GalliumConcreteA_FaceAcc_AnimBP
target_skeleton=/BRCosmetics/Characters/Player/Female/Medium/Bodies/F_MED_Vampire_Casual/Meshes/Parts/F_MED_Vampire_Casual_Skeleton
graphs [
 AnimGraph {nodes [
   n1 = "Transform (Modify) Bone - Bone: dyn_hair_f_MSTR"(
    ComponentPose=Local To Component(LocalPose=Use cached pose 'AnimDynamics_Cache'())
    Alpha=Get headUp() -> (headUp:double)
    Translation="-2,0,0"
    Rotation="20,0,0"
    Scale="1,1,1"
   ) {
    rotationMode=BMM_Additive
    rotationSpace=BCS_BoneSpace
    scaleMode=BMM_Additive
    scaleSpace=BCS_BoneSpace
    translationMode=BMM_Additive
    translationSpace=BCS_BoneSpace
    alphaBoolBlend {blendOption=Linear}
    boneToModify {boneName=dyn_hair_f_MSTR}
    rotation {pitch=20}
    translation {x=-2}
   }
   n2 = Get GravityOverride()
   n3 = RigidBody!(ComponentPose=@n1, Alpha=Get MeleeDynamics() -> (MeleeDynamics:float), OverrideWorldGravity:Vector=@n2, ExternalForce) -> (Pose) {
    bClampLinearTranslationLimitToRefPose=true
    bOverrideWorldGravity=true
    overridePhysicsAsset=&/F_MED_GalliumConcreteA_Hair_Physics
    simulationSpace=BaseBoneSpace
    alphaBoolBlend {blendOption=Linear}
    alphaScaleBiasClamp {bInterpResult=true}
    baseBoneRef {boneName=neck_01}
    componentAppliedLinearAccClamp=<3000 3000 3000>
    componentLinearAccScale=<0.3 0.3 0.3>
    componentLinearVelScale=<0.3 0.3 0.3>
    overrideWorldGravity {z=-980}
   }
   n4 = Component To Local(ComponentPose=@n3)
   n5 = RigidBody(ComponentPose=Local To Component(LocalPose=Use cached pose 'AnimDynamics_Cache'()), OverrideWorldGravity=@n2) {
    bClampLinearTranslationLimitToRefPose=true
    bOverrideWorldGravity=true
    overridePhysicsAsset=&/F_MED_GalliumConcreteA_Hair_FrontEnd_Physics
    simulationSpace=BaseBoneSpace
    alphaBoolBlend {blendOption=Linear}
    baseBoneRef {boneName=neck_01}
    componentAppliedLinearAccClamp=<3000 3000 3000>
    componentLinearAccScale=<0.3 0.3 0.3>
    componentLinearVelScale=<0.3 0.3 0.3>
    overrideWorldGravity {z=-980}
   }
   n6 = Component To Local(ComponentPose=@n5)
   n7 = Get bIsInFrontEnd()
   n8 = NOT Boolean(A=Get bRigidBodyEnabled())
   n9 = Get Is Wearing Cape() -> (Is Wearing Cape:bool)
   n10 = Get isWearingBackbling()
   Output Pose(Result=Blend Poses by bool(
     BlendPose_0=Use cached pose 'constraintHair_Cache'()
     BlendPose_1=Blend Poses by bool(BlendPose_0=@n6, BlendPose_1=@n4, bActiveValue=@n7) {blendType=Linear}
     bActiveValue=OR Boolean!(A=@n10, B=@n9, C:bool=@n8) -> (ReturnValue)
    ) {blendType=Linear}) {layerGroup=None}
   n11 = Control Rig!(
    Source=Copy Pose From Mesh() {bUseAttachedParent=true}
    Alpha
    MeleeAttack_out_deform:double=Get MeleeAttack() -> (MeleeAttack:double)
    Skydive_out_deform:double=Get IsSkydive() -> (IsSkydive:double)
    wearing Backpack:bool=Get isWearingBackbling()
   ) -> (Pose) {
    alphaBoolBlend {blendOption=Linear}
    controlRigAssetReference {blueprintRigClass=&/F_MED_GalliumConcreteA_Hair_CtrlRig.F_MED_GalliumConcreteA_Hair_CtrlRig_C}
   }
   Save cached pose 'PreAim'(Pose=@n11) {cacheName=PreAim}
   n12 = RigidBody(ComponentPose=Local To Component(LocalPose=Use cached pose 'AnimDynamics_Cache'()), OverrideWorldGravity=Get GravityOverride()) {
    bClampLinearTranslationLimitToRefPose=true
    bOverrideWorldGravity=true
    overridePhysicsAsset=&F_MED_GalliumConcreteA_Hair_Physics
    simulationSpace=BaseBoneSpace
    alphaBoolBlend {blendOption=Linear}
    alphaScaleBiasClamp {bInterpResult=true}
    baseBoneRef {boneName=neck_01}
    componentAppliedLinearAccClamp=<3000 3000 3000>
    componentLinearAccScale=<0.3 0.3 0.3>
    componentLinearVelScale=<0.3 0.3 0.3>
    overrideWorldGravity {z=-980}
   }
   n13 = "Constraint - dyn_hair_a_02 (P)"(ComponentPose=@n12, Alpha=0.1, ConstraintWeights_0=1) {
    alphaBoolBlend {blendOption=Linear}
    boneToModify {boneName=dyn_hair_a_02}
    constraintSetup [{targetBone {boneName=clavicle_l}}]
    constraintWeights [1]
   }
   n14 = "Constraint - dyn_hair_b_01 (P)"(ComponentPose=@n13, Alpha=0.2, ConstraintWeights_0=1) {
    alphaBoolBlend {blendOption=Linear}
    boneToModify {boneName=dyn_hair_b_01}
    constraintSetup [{targetBone {boneName=clavicle_l}}]
    constraintWeights [1]
   }
   n15 = "Constraint - dyn_hair_a_03 (P)"(ComponentPose=@n14, Alpha=0.2, ConstraintWeights_0=1) {
    alphaBoolBlend {blendOption=Linear}
    boneToModify {boneName=dyn_hair_a_03}
    constraintSetup [{targetBone {boneName=clavicle_l}}]
    constraintWeights [1]
   }
   n16 = "Constraint - dyn_hair_b_02 (P)"(ComponentPose=@n15, Alpha=0.3, ConstraintWeights_0=1) {
    alphaBoolBlend {blendOption=Linear}
    boneToModify {boneName=dyn_hair_b_02}
    constraintSetup [{targetBone {boneName=clavicle_l}}]
    constraintWeights [1]
   }
   n17 = "Constraint - dyn_hair_a_04 (P)"(ComponentPose=@n16, Alpha=0.3, ConstraintWeights_0=1) {
    alphaBoolBlend {blendOption=Linear}
    boneToModify {boneName=dyn_hair_a_04}
    constraintSetup [{targetBone {boneName=clavicle_l}}]
    constraintWeights [1]
   }
   n18 = "Constraint - dyn_hair_a_05 (P)"(ComponentPose=@n17, Alpha=0.4, ConstraintWeights_0=1) {
    alphaBoolBlend {blendOption=Linear}
    boneToModify {boneName=dyn_hair_a_05}
    constraintSetup [{targetBone {boneName=clavicle_l}}]
    constraintWeights [1]
   }
   n19 = "Constraint - dyn_hair_c_02 (P)"(ComponentPose=@n18, Alpha=0.8, ConstraintWeights_0=1) {
    alphaBoolBlend {blendOption=Linear}
    boneToModify {boneName=dyn_hair_c_02}
    constraintSetup [{targetBone {boneName=clavicle_l}}]
    constraintWeights [1]
   }
   n20 = "Constraint - dyn_hair_c_03 (PP)"!(ComponentPose=@n19, Alpha=0.8, ConstraintWeights_0=0.2, ConstraintWeights_1:float=0.8) -> (Pose) {
    alphaBoolBlend {blendOption=Linear}
    boneToModify {boneName=dyn_hair_c_03}
    constraintSetup [{targetBone {boneName=clavicle_l}}, {targetBone {boneName=spine_05}}]
    constraintWeights [1,1]
   }
   n21 = "Constraint - dyn_hair_d_02 (PP)"!(ComponentPose=@n20, Alpha=1, ConstraintWeights_0=0.5, ConstraintWeights_1:float=0.5) -> (Pose) {
    alphaBoolBlend {blendOption=Linear}
    boneToModify {boneName=dyn_hair_d_02}
    constraintSetup [{targetBone {boneName=neck_01}}, {targetBone {boneName=spine_05}}]
    constraintWeights [1,1]
   }
   n22 = "Constraint - dyn_hair_d_03 (PP)"!(ComponentPose=@n21, Alpha=1, ConstraintWeights_0=0.2, ConstraintWeights_1:float=0.8) -> (Pose) {
    alphaBoolBlend {blendOption=Linear}
    boneToModify {boneName=dyn_hair_d_03}
    constraintSetup [{targetBone {boneName=clavicle_l}}, {targetBone {boneName=spine_05}}]
    constraintWeights [1,1]
   }
   n23 = "Constraint - dyn_hair_d_04 (PP)"!(ComponentPose=@n22, Alpha=1, ConstraintWeights_0=0.5, ConstraintWeights_1:float=0.5) -> (Pose) {
    alphaBoolBlend {blendOption=Linear}
    boneToModify {boneName=dyn_hair_d_04}
    constraintSetup [{targetBone {boneName=spine_04}}, {targetBone {boneName=spine_05}}]
    constraintWeights [1,1]
   }
   n24 = "Constraint - dyn_hair_d_05 (PP)"!(ComponentPose=@n23, Alpha=1, ConstraintWeights_0=0.1, ConstraintWeights_1:float=0.9) -> (Pose) {
    alphaBoolBlend {blendOption=Linear}
    boneToModify {boneName=dyn_hair_d_05}
    constraintSetup [{targetBone {boneName=spine_05}}, {targetBone {boneName=spine_04}}]
    constraintWeights [1,1]
   }
   n25 = "Constraint - dyn_hair_e_02 (PP)"!(ComponentPose=@n24, Alpha=1, ConstraintWeights_0=0.5, ConstraintWeights_1:float=0.5) -> (Pose) {
    alphaBoolBlend {blendOption=Linear}
    boneToModify {boneName=dyn_hair_e_02}
    constraintSetup [{targetBone {boneName=spine_05}}, {targetBone {boneName=neck_01}}]
    constraintWeights [1,1]
   }
   n26 = "Constraint - dyn_hair_e_03 (P)"(ComponentPose=@n25, Alpha=1, ConstraintWeights_0=1) {
    alphaBoolBlend {blendOption=Linear}
    boneToModify {boneName=dyn_hair_e_03}
    constraintSetup [{targetBone {boneName=spine_05}}]
    constraintWeights [1]
   }
   n27 = "Constraint - dyn_hair_e_04 (PP)"!(ComponentPose=@n26, Alpha=1, ConstraintWeights_0=0.5, ConstraintWeights_1:float=0.5) -> (Pose) {
    alphaBoolBlend {blendOption=Linear}
    boneToModify {boneName=dyn_hair_e_04}
    constraintSetup [{targetBone {boneName=spine_05}}, {targetBone {boneName=spine_04}}]
    constraintWeights [1,1]
   }
   n28 = "Constraint - dyn_hair_e_05 (PP)"!(ComponentPose=@n27, Alpha=1, ConstraintWeights_0=0.1, ConstraintWeights_1:float=0.9) -> (Pose) {
    alphaBoolBlend {blendOption=Linear}
    boneToModify {boneName=dyn_hair_e_05}
    constraintSetup [{targetBone {boneName=spine_05}}, {targetBone {boneName=spine_04}}]
    constraintWeights [1,1]
   }
   n29 = "Constraint - dyn_hair_f_02 (PP)"!(ComponentPose=@n28, Alpha=1, ConstraintWeights_0=0.5, ConstraintWeights_1:float=0.5) -> (Pose) {
    alphaBoolBlend {blendOption=Linear}
    boneToModify {boneName=dyn_hair_f_02}
    constraintSetup [{targetBone {boneName=neck_01}}, {targetBone {boneName=spine_05}}]
    constraintWeights [1,1]
   }
   n30 = "Constraint - dyn_hair_f_03 (PP)"!(ComponentPose=@n29, Alpha=1, ConstraintWeights_0=0.2, ConstraintWeights_1:float=0.8) -> (Pose) {
    alphaBoolBlend {blendOption=Linear}
    boneToModify {boneName=dyn_hair_f_03}
    constraintSetup [{targetBone {boneName=clavicle_r}}, {targetBone {boneName=spine_05}}]
    constraintWeights [1,1]
   }
   n31 = "Constraint - dyn_hair_f_04 (PP)"!(ComponentPose=@n30, Alpha=1, ConstraintWeights_0=0.5, ConstraintWeights_1:float=0.5) -> (Pose) {
    alphaBoolBlend {blendOption=Linear}
    boneToModify {boneName=dyn_hair_f_04}
    constraintSetup [{targetBone {boneName=spine_04}}, {targetBone {boneName=spine_05}}]
    constraintWeights [1,1]
   }
   n32 = "Constraint - dyn_hair_f_05 (PP)"!(ComponentPose=@n31, Alpha=1, ConstraintWeights_0=0.1, ConstraintWeights_1:float=0.9) -> (Pose) {
    alphaBoolBlend {blendOption=Linear}
    boneToModify {boneName=dyn_hair_f_05}
    constraintSetup [{targetBone {boneName=spine_05}}, {targetBone {boneName=spine_04}}]
    constraintWeights [1,1]
   }
   n33 = "Constraint - dyn_hair_g_02 (P)"(ComponentPose=@n32, Alpha=0.2, ConstraintWeights_0=1) {
    alphaBoolBlend {blendOption=Linear}
    boneToModify {boneName=dyn_hair_g_02}
    constraintSetup [{targetBone {boneName=clavicle_r}}]
    constraintWeights [1]
   }
   n34 = "Constraint - dyn_hair_g_03 (P)"(ComponentPose=@n33, Alpha=0.3, ConstraintWeights_0=1) {
    alphaBoolBlend {blendOption=Linear}
    boneToModify {boneName=dyn_hair_g_03}
    constraintSetup [{targetBone {boneName=clavicle_r}}]
    constraintWeights [1]
   }
   n35 = "Constraint - dyn_hair_h_02 (P)"(ComponentPose=@n34, Alpha=0.2, ConstraintWeights_0=1) {
    alphaBoolBlend {blendOption=Linear}
    boneToModify {boneName=dyn_hair_h_02}
    constraintSetup [{targetBone {boneName=clavicle_r}}]
    constraintWeights [1]
   }
   n36 = "Constraint - dyn_hair_h_03 (P)"(ComponentPose=@n35, Alpha=0.3, ConstraintWeights_0=1) {
    alphaBoolBlend {blendOption=Linear}
    boneToModify {boneName=dyn_hair_h_03}
    constraintSetup [{targetBone {boneName=clavicle_r}}]
    constraintWeights [1]
   }
   n37 = "Constraint - dyn_hair_h_04 (P)"(ComponentPose=@n36, Alpha=0.4, ConstraintWeights_0=1) {
    alphaBoolBlend {blendOption=Linear}
    boneToModify {boneName=dyn_hair_h_04}
    constraintSetup [{targetBone {boneName=clavicle_r}}]
    constraintWeights [1]
   }
   Save cached pose 'constraintHair_Cache'(Pose=Component To Local(ComponentPose=@n37)) {cacheName=constraintHair_Cache}
   n38 = Anim Dynamics - Bone: dyn_hat_01(ComponentPose=Local To Component(LocalPose=Use cached pose 'PreAim'()), Alpha=0.7, ExternalForce="0,0,0") {
    angularDampingOverride=0.9999
    angularSpringConstant=1500
    bAngularSpring=true
    bGravityOverrideInSimSpace=true
    bOverrideAngularDamping=true
    simulationSpace=RootRelative
    alphaBoolBlend {blendOption=Linear}
    boundBone {boneName=dyn_hat_01}
    componentLinearAccScale=<1 1 1>
    componentLinearVelScale=<1 1 1>
    physicsBodyDefinitions [{
      sphereCollisionRadius=0
      boundBone {boneName=dyn_hat_01}
      boxExtents=<20 20 10>
      constraintSetup {angularLimitsMax=<20 2 20>, angularLimitsMin=<-20 -45 -45>, angularTarget=<0 0 0>}
      localJointOffset=<0 0 -2>
     }]
   }
   n39 = Anim Dynamics - Bone: dyn_hat_02(ComponentPose=@n38, Alpha=0.7, ExternalForce="0,0,0") {
    angularDampingOverride=0.9999
    angularSpringConstant=5000
    bAngularSpring=true
    bGravityOverrideInSimSpace=true
    bOverrideAngularDamping=true
    simulationSpace=RootRelative
    alphaBoolBlend {blendOption=Linear}
    boundBone {boneName=dyn_hat_02}
    componentLinearAccScale=<1 1 1>
    componentLinearVelScale=<1 1 1>
    physicsBodyDefinitions [{
      sphereCollisionRadius=0
      boundBone {boneName=dyn_hat_02}
      boxExtents=<20 20 10>
      constraintSetup {angularLimitsMax=<20 2 20>, angularLimitsMin=<-20 -45 -45>, angularTarget=<0 0 0>}
      localJointOffset=<0.5 0 -2>
     }]
   }
   n40 = Anim Dynamics - Bone: dyn_headBandDeco(ComponentPose=@n39, Alpha=1, ExternalForce="0,0,0") {
    angularSpringConstant=250
    bAngularSpring=true
    bGravityOverrideInSimSpace=true
    bOverrideAngularDamping=true
    gravityScale=0.5
    alphaBoolBlend {blendOption=Linear}
    boundBone {boneName=dyn_headBandDeco}
    physicsBodyDefinitions [{
      sphereCollisionRadius=0
      boundBone {boneName=dyn_headBandDeco}
      boxExtents=<10 10 5>
      constraintSetup {angularLimitsMax=<20 1 5>, angularLimitsMin=<-20 -1 -5>, angularTarget=<0 0 0>}
      localJointOffset=<0 0 -3>
     }]
   }
   n41 = Anim Dynamics - Bone: dyn_bangs_a_01(ComponentPose=@n40, Alpha=1, ExternalForce="0,0,0") {
    angularDampingOverride=0.9999
    angularSpringConstant=75
    bAngularSpring=true
    bGravityOverrideInSimSpace=true
    bOverrideAngularDamping=true
    simulationSpace=RootRelative
    alphaBoolBlend {blendOption=Linear}
    boundBone {boneName=dyn_bangs_a_01}
    componentLinearAccScale=<1 1 1>
    componentLinearVelScale=<1 1 1>
    physicsBodyDefinitions [{
      sphereCollisionRadius=0
      boundBone {boneName=dyn_bangs_a_01}
      boxExtents=<20 20 10>
      constraintSetup {angularLimitsMax=<25 0 15>, angularLimitsMin=<-25 -25 -15>, angularTarget=<0 0 0>}
      localJointOffset=<2 0 0>
     }]
   }
   n42 = Anim Dynamics - Bone: dyn_bangs_a_02(ComponentPose=@n41, Alpha=1, ExternalForce="0,0,0") {
    angularDampingOverride=0.9999
    angularSpringConstant=75
    bAngularSpring=true
    bGravityOverrideInSimSpace=true
    bOverrideAngularDamping=true
    simulationSpace=RootRelative
    alphaBoolBlend {blendOption=Linear}
    boundBone {boneName=dyn_bangs_a_02}
    componentLinearAccScale=<1 1 1>
    componentLinearVelScale=<1 1 1>
    physicsBodyDefinitions [{
      sphereCollisionRadius=0
      boundBone {boneName=dyn_bangs_a_02}
      boxExtents=<20 20 10>
      constraintSetup {angularLimitsMax=<25 0 15>, angularLimitsMin=<-25 -25 -15>, angularTarget=<0 0 0>}
      localJointOffset=<2 0 0>
     }]
   }
   n43 = Anim Dynamics - Bone: dyn_bangs_b_01(ComponentPose=@n42, Alpha=1, ExternalForce="0,0,0") {
    angularDampingOverride=0.9999
    angularSpringConstant=75
    bAngularSpring=true
    bGravityOverrideInSimSpace=true
    bOverrideAngularDamping=true
    simulationSpace=RootRelative
    alphaBoolBlend {blendOption=Linear}
    boundBone {boneName=dyn_bangs_b_01}
    componentLinearAccScale=<1 1 1>
    componentLinearVelScale=<1 1 1>
    physicsBodyDefinitions [{
      sphereCollisionRadius=0
      boundBone {boneName=dyn_bangs_b_01}
      boxExtents=<20 20 10>
      constraintSetup {angularLimitsMax=<25 0 15>, angularLimitsMin=<-25 -25 -15>, angularTarget=<0 0 0>}
      localJointOffset=<2 0 0>
     }]
   }
   n44 = Anim Dynamics - Bone: dyn_bangs_b_02(ComponentPose=@n43, Alpha=1, ExternalForce="0,0,0") {
    angularDampingOverride=0.9999
    angularSpringConstant=75
    bAngularSpring=true
    bGravityOverrideInSimSpace=true
    bOverrideAngularDamping=true
    simulationSpace=RootRelative
    alphaBoolBlend {blendOption=Linear}
    boundBone {boneName=dyn_bangs_b_02}
    componentLinearAccScale=<1 1 1>
    componentLinearVelScale=<1 1 1>
    physicsBodyDefinitions [{
      sphereCollisionRadius=0
      boundBone {boneName=dyn_bangs_b_02}
      boxExtents=<20 20 10>
      constraintSetup {angularLimitsMax=<25 0 15>, angularLimitsMin=<-25 -25 -15>, angularTarget=<0 0 0>}
      localJointOffset=<2 0 0>
     }]
   }
   n45 = Anim Dynamics - Bone: dyn_bangs_c_01(ComponentPose=@n44, Alpha=1, ExternalForce="0,0,0") {
    angularDampingOverride=0.9999
    angularSpringConstant=75
    bAngularSpring=true
    bGravityOverrideInSimSpace=true
    bOverrideAngularDamping=true
    simulationSpace=RootRelative
    alphaBoolBlend {blendOption=Linear}
    boundBone {boneName=dyn_bangs_c_01}
    componentLinearAccScale=<1 1 1>
    componentLinearVelScale=<1 1 1>
    physicsBodyDefinitions [{
      sphereCollisionRadius=0
      boundBone {boneName=dyn_bangs_c_01}
      boxExtents=<20 20 10>
      constraintSetup {angularLimitsMax=<25 0 15>, angularLimitsMin=<-25 -25 -15>, angularTarget=<0 0 0>}
      localJointOffset=<2 0 0>
     }]
   }
   Save cached pose 'AnimDynamics_Cache'(Pose=Component To Local(ComponentPose=@n45)) {cacheName=AnimDynamics_Cache}
  ]}
 EventGraph {nodes [
   Event BlueprintUpdateAnimation()
   > n1 = Sequence!(execute) -> (then_0, then_1, then_2:Exec, then_3:Exec)
   n2 = Gravity Override State Function DEPRECATED(
    execute=@n1.then_1
    CharacterPartAnimInstance=Self-Reference()
    GameTimeInSeconds=GetGameTimeInSeconds()
    GravityOverrideParams=Get Gravity Override Params()
   )
   > Set GravityOverride(GravityOverride=@n2)
   > Set headUp(headUp:double=GetAlphaFromReferenceTransformDelta(SkeletalMeshComponent=Get Skel Mesh(), BoneName=spine_05, BaseName=head, InputRange=(X=0,Y=60), TwistAxis="0,0,1")) -> (Output_Get:double)
   n3 = BreakTransform(InTransform=GetDeltaTransformFromRefPose(self=GetOwnerSkeletalMeshComponent(), BoneName=head, BaseName=neck_02))
   > Set RifleLookUp(RifleLookUp:double=MapRangeClamped(Value=BreakRotator(InRot=@n3.Rotation).Yaw, InRangeA=5, InRangeB=-20)) -> (Output_Get:double)
   n4 = Get bIsPlayingMeleeAnim()
   Set MeleeAttack(execute=@n1.then_3, MeleeAttack:double="To Float (Boolean)"(InBool=@n4)) -> (Output_Get:double)
   n5 = Get bIsSkyDiving()
   n6 = Get bIsSkydiveFloating()
   > Set IsSkydive(IsSkydive:double="To Float (Boolean)"(InBool=AND Boolean(A=@n5, B=@n6))) -> (Output_Get:double)
   n7 = Get isWearingBackbling()
   > n8 = Branch(Condition=AND Boolean(A=@n4, B=NOT Boolean(A=@n7)))
   > Set MeleeDynamics(MeleeDynamics:float=0.6) -> (Output_Get:float)
   Set MeleeDynamics(execute=@n8.else, MeleeDynamics=1)
   Set Crouching(execute=@n1.then_0, Crouching:bool=Get bIsCrouching()) -> (Output_Get:bool)
   > Set Skydiving(Skydiving:bool=OR Boolean(A=Get bIsSkyDiving(), B=Get bIsSkydiveFloating(), C=Get bIsSkydiveDiveMode())) -> (Output_Get:bool)
   n9 = CheckForBackbling()
   Set isWearingBackbling(
    execute=@n1.then_2
    isWearingBackbling=AND Boolean!(A=@n9, B=NOT Boolean(A=@n9.bIsBackblingACape), C:bool=NOT Boolean(A=@n9.bIsBackblingLong)) -> (ReturnValue)
   )
   > Set Is Wearing Cape(Is Wearing Cape=OR Boolean!(A=@n9.bIsBackblingACape, B=@n9.bIsBackblingLong) -> (ReturnValue))
   Event BlueprintInitializeAnimation()
   > Set Skel Mesh(Skel Mesh=GetOwnerSkeletalMeshComponent())
  ]}
]
variables [
 Crouching {type=boolean}
 Gravity Override Params {type=GravityOverrideParamsStruct}
 GravityOverride {type=Vector}
 headUp {type=number}
 Is Wearing Cape {type=boolean}
 IsSkydive {type=number}
 isWearingBackbling {type=boolean}
 MeleeAlpha {type=number}
 MeleeAttack {type=number}
 MeleeDynamics {type=number, default=0.7}
 RifleLookUp {type=number}
 Skel Mesh {type=SkeletalMeshComponent}
 Skydiving {type=boolean}
]
